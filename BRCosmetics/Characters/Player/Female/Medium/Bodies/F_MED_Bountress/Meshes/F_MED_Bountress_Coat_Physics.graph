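# Physics asset for the F_MED_Bountress female player body's coat, defining collision and joints for its dynamic coat bones.
# Coat panels (front, back, left-middle) and a loop are limited swing/twist chains hanging from tiny base bodies constrained to the pelvis.
# Capsules on spine, arms and legs, plus large spheres on root and extra_ground, provide body and ground collision shapes.
# An angularly free constraint cross-links the left front and middle panels.
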PhysicsAsset /BRCosmetics/Characters/Player/Female/Medium/Bodies/F_MED_Bountress/Meshes/F_MED_Bountress_Coat_Physics
bodies {
 calf_l [{shapeType=Capsule, length=32.9844, radius=7.0156, shapeName=calf_l_capsule, center=<-19.4567 -2.431 -0.5072>, rotation=<rot 90 0 0>}]
 calf_r [
  {shapeType=Capsule, length=38.9082, radius=6.0918, shapeName=calf_r_capsule, center=<25.2816 2.6863 0.6944>, rotation=<rot 90 0 0>}
  {
   shapeType=Capsule
   length=21.9866
   radius=7.3485
   shapeName=calf_r_capsule
   center=<39.0704 4.7079 0.151>
   rotation=<rot 1.7676 105.8338 90.5012>
  }
 ]
 dyn_coat_01_bk_l [{shapeType=Capsule, length=5.2136, radius=15, shapeName=dyn_coat_01_frnt_r_capsule, center=<9.254 -7.5194 -10.1125>, rotation=<rot 90 0 0>}]
 dyn_coat_01_frnt_l [{
   shapeType=Box
   extentX=14.4213
   extentY=22.6661
   extentZ=4.4016
   shapeName=dyn_coat_01_frnt_l_box
   center=<7.2984 2.4012 0.4141>
   rotation=<rot -0.0002 -0.0003 3.477>
  }]
 dyn_coat_01_frnt_r [{
   shapeType=Box
   extentX=13.5992
   extentY=12.9957
   extentZ=2.2796
   shapeName=dyn_coat_01_frnt_r_box
   center=<7.0566 0 0>
   rotation=<rot 0.0001 0.0001 18.2145>
  }]
 dyn_coat_01_mid_l [{
   shapeType=Capsule
   length=13.3949
   radius=6.9755
   shapeName=dyn_coat_01_mid_l_capsule
   center=<6.7465 0.1047 -5.365>
   rotation=<rot 77.5711 78.9617 169.563>
  }]
 dyn_coat_02_frnt_l [{
   shapeType=Box
   extentX=20.1469
   extentY=30.6904
   extentZ=5.2996
   shapeName=dyn_coat_02_frnt_l_box
   center=<13.7392 0.2906 0.4488>
   rotation=<rot 5.0034 40.2694 -4.5614>
  }]
 dyn_coat_02_frnt_r [{
   shapeType=Box
   extentX=11.2118
   extentY=16.3253
   extentZ=2.3335
   shapeName=dyn_coat_02_frnt_r_box
   center=<6.3128 0.1211 0>
   rotation=<rot 0.0001 0.0001 18.2145>
  }]
 dyn_coat_02_mid_l [{
   shapeType=Capsule
   length=14.9508
   radius=6.9755
   shapeName=dyn_coat_02_mid_l_capsule
   center=<7.3887 -0.0011 -5.7178>
   rotation=<rot 77.2642 73.3651 163.8347>
  }]
 dyn_coat_03_frnt_r [
  {
   shapeType=Capsule
   length=14.5494
   radius=6.044
   shapeName=dyn_coat_03_frnt_r_capsule
   center=<9.7089 10.949 -0.327>
   rotation=<rot 85.0001 0.0002 0.0001>
  }
  {shapeType=Box, extentX=17.9204, extentY=16.5954, extentZ=2.7754, shapeName=dyn_coat_03_frnt_r_box, center=<10.6368 0 0.5128>}
 ]
 dyn_coat_1_bk_r [{shapeType=Capsule, length=6.3931, radius=7.2533, shapeName=dyn_coat_01_frnt_r_capsule, center=<9.2379 0 -1.2628>, rotation=<rot 90 0 0>}]
 dyn_coat_base_bk_l [{shapeType=Sphere, radius=0.5, shapeName=dyn_coat_base_bk_l_sphere}]
 dyn_coat_base_bk_r [{shapeType=Sphere, radius=0.5, shapeName=dyn_coat_base_bk_r_sphere}]
 dyn_coat_base_frnt_l [{shapeType=Sphere, radius=0.5, shapeName=dyn_coat_base_frnt_l_sphere}]
 dyn_coat_base_frnt_r [{shapeType=Sphere, radius=0.5, shapeName=dyn_coat_base_frnt_r_sphere}]
 dyn_coat_base_mid_l [{shapeType=Sphere, radius=0.5, shapeName=dyn_coat_base_mid_l_sphere}]
 dyn_coat_loop_base [{shapeType=Capsule, length=0.505, radius=0.505, shapeName=dyn_coat_loop_base_capsule, rotation=<rot 0 -90 0>}]
 dyn_coat_loop_parent [{
   shapeType=Capsule
   length=12
   radius=1.5
   shapeName=dyn_coat_loop_parent_capsule
   center=<7.6035 0 0.0263>
   rotation=<rot 81.4105 -179.9989 -179.9984>
  }]
 dyn_coat_parent_bk_l1 [{shapeType=Capsule, length=8.8802, radius=3.1853, shapeName=dyn_coat_parent_frnt_r_capsule, center=<6.1032 0 0>, rotation=<rot 90 0 0>}]
 dyn_coat_parent_bk_r [{shapeType=Capsule, length=7.9361, radius=4.1294, shapeName=dyn_coat_parent_frnt_r_capsule, center=<6.1032 0 -1.5167>, rotation=<rot 90 0 0>}]
 dyn_coat_parent_frnt_l1 [{
   shapeType=Box
   extentX=23.1066
   extentY=21.8576
   extentZ=6.2312
   shapeName=dyn_coat_parent_frnt_l1_box
   center=<11.962 2.3032 0.1351>
   rotation=<rot -0.0003 -0.0003 -5.3399>
  }]
 dyn_coat_parent_frnt_r [{shapeType=Capsule, length=8.8802, radius=3.1853, shapeName=dyn_coat_parent_frnt_r_capsule, center=<6.1032 0 0>, rotation=<rot 90 0 0>}]
 dyn_coat_parent_mid_l [{
   shapeType=Capsule
   length=10.4493
   radius=6.9755
   shapeName=dyn_coat_parent_mid_l_capsule
   center=<5.7676 -0.2218 -5.9735>
   rotation=<rot 77.526 77.9708 168.5455>
  }]
 extra_ground [
  {shapeType=Sphere, radius=30.13, shapeName=root_sphere, center=<0.0002 -7.2623 8.5972>}
  {shapeType=Sphere, radius=175, shapeName=root_sphere, center=<0 -8.6403 -165.6077>}
 ]
 foot_l [{shapeType=Capsule, length=26, radius=7.5, shapeName=foot_r_capsule, center=<-4.3789 2.9446 0>, rotation=<rot 0 0 90>}]
 foot_r [{shapeType=Capsule, length=26, radius=7.5, shapeName=foot_r_capsule, center=<3.3697 -5.0998 0>, rotation=<rot 0 0 90>}]
 lowerarm_l [{
   shapeType=Capsule
   length=18.4966
   radius=5.1795
   shapeName=lowerarm_l_capsule
   center=<11.8221 -0.908 -0.3383>
   rotation=<rot 0 -90 -88.361>
  }]
 lowerarm_r [{
   shapeType=Capsule
   length=18.4966
   radius=5.1795
   shapeName=lowerarm_l_capsule
   center=<-12.3205 -0.908 0.3525>
   rotation=<rot 0 -90 -88.361>
  }]
 pelvis [{shapeType=Capsule, length=1.0593, radius=11.5292, shapeName=pelvis_capsule, center=<0 -1.4312 0>}]
 root [
  {shapeType=Sphere, radius=30.13, shapeName=root_sphere, center=<0.0002 -7.2623 8.5972>}
  {shapeType=Sphere, radius=175, shapeName=root_sphere, center=<0 -8.6403 -165.6077>}
 ]
 spine_01 [{shapeType=Capsule, length=0.505, radius=0.505, shapeName=spine_01_capsule, rotation=<rot 0 -90 0>}]
 spine_02 [{
   shapeType=Capsule
   length=3.5886
   radius=10.2932
   shapeName=spine_02_capsule
   center=<1.3313 0.2931 -0.4397>
   rotation=<rot 0 -169.0228 -156.8696>
  }]
 spine_03 [{shapeType=Capsule, length=0.505, radius=0.505, shapeName=spine_03_capsule, center=<1.1754 -7.1962 0.0007>, rotation=<rot 0.0091 -90 -0.0408>}]
 spine_04 [{
   shapeType=Capsule
   length=4.167
   radius=11.7862
   shapeName=spine_04_capsule
   center=<0.2912 1.564 0.1717>
   rotation=<rot -12.1921 -90 -1.484>
  }]
 spine_05 [{shapeType=Capsule, length=13.1809, radius=10.9612, shapeName=spine_05_capsule}]
 thigh_twist_01_l [{
   shapeType=Capsule
   length=26.2216
   radius=7.7784
   shapeName=thigh_twist_01_l_capsule
   center=<4.7138 -0.8436 0.8044>
   rotation=<rot 85 179.9997 179.9999>
  }]
 thigh_twist_01_r []
 upperarm_l [{
   shapeType=Capsule
   length=26.2114
   radius=6.5265
   shapeName=upperarm_l_capsule
   center=<13.2949 -1.081 -0.2242>
   rotation=<rot 3.4744 -87.5326 -90.0942>
  }]
 upperarm_r [{
   shapeType=Capsule
   length=22.4586
   radius=5.734
   shapeName=upperarm_l_capsule
   center=<-10.0218 1.8218 0.5701>
   rotation=<rot 3.4744 -87.5326 -90.0942>
  }]
}
constraints [
 {bone1Name=dyn_coat_01_mid_l, bone2Name=dyn_coat_parent_mid_l, swing1LimitDegrees=20, swing2LimitDegrees=20, twistLimitDegrees=3}
 {bone1Name=dyn_coat_parent_mid_l, bone2Name=dyn_coat_base_mid_l, swing1LimitDegrees=20, swing2LimitDegrees=50, twistLimitDegrees=5}
 {bone1Name=dyn_coat_02_frnt_r, bone2Name=dyn_coat_01_frnt_r, swing1LimitDegrees=5, swing2LimitDegrees=20, twistLimitDegrees=3}
 {bone1Name=dyn_coat_01_frnt_r, bone2Name=dyn_coat_parent_frnt_r, swing1LimitDegrees=5, swing2LimitDegrees=20, twistLimitDegrees=3}
 {bone1Name=dyn_coat_parent_frnt_r, bone2Name=dyn_coat_base_frnt_r, swing1LimitDegrees=5, swing2LimitDegrees=20, twistLimitDegrees=3}
 {bone1Name=dyn_coat_03_frnt_r, bone2Name=dyn_coat_02_frnt_r, swing1LimitDegrees=5, swing2LimitDegrees=20, twistLimitDegrees=3}
 {bone1Name=dyn_coat_02_mid_l, bone2Name=dyn_coat_01_mid_l, swing1LimitDegrees=20, swing2LimitDegrees=20, twistLimitDegrees=3}
 {bone1Name=dyn_coat_1_bk_r, bone2Name=dyn_coat_parent_bk_r, swing1LimitDegrees=5, swing2LimitDegrees=20, twistLimitDegrees=3}
 {bone1Name=dyn_coat_parent_bk_r, bone2Name=dyn_coat_base_bk_r, swing1LimitDegrees=5, swing2LimitDegrees=20, twistLimitDegrees=3}
 {bone1Name=dyn_coat_01_bk_l, bone2Name=dyn_coat_parent_bk_l1, swing1LimitDegrees=5, swing2LimitDegrees=20, twistLimitDegrees=3}
 {bone1Name=dyn_coat_parent_bk_l1, bone2Name=dyn_coat_base_bk_l, swing1LimitDegrees=5, swing2LimitDegrees=20, twistLimitDegrees=3}
 {
  bone1Name=foot_l
  bone2Name=calf_l
  swing1LimitDegrees=45
  swing1Motion=Free
  swing2LimitDegrees=45
  swing2Motion=Free
  twistLimitDegrees=45
  twistMotion=Free
 }
 {
  bone1Name=foot_r
  bone2Name=calf_r
  swing1LimitDegrees=45
  swing1Motion=Free
  swing2LimitDegrees=45
  swing2Motion=Free
  twistLimitDegrees=45
  twistMotion=Free
 }
 {bone1Name=dyn_coat_parent_frnt_l1, bone2Name=dyn_coat_base_frnt_l, swing1LimitDegrees=100, swing2LimitDegrees=100, twistLimitDegrees=35}
 {bone1Name=dyn_coat_01_frnt_l, bone2Name=dyn_coat_parent_frnt_l1, swing1LimitDegrees=20, swing2LimitDegrees=20, twistLimitDegrees=5}
 {bone1Name=dyn_coat_02_frnt_l, bone2Name=dyn_coat_01_frnt_l, swing1LimitDegrees=20, swing2LimitDegrees=20, twistLimitDegrees=5}
 {
  bone1Name=calf_l
  bone2Name=pelvis
  swing1LimitDegrees=45
  swing1Motion=Free
  swing2LimitDegrees=45
  swing2Motion=Free
  twistLimitDegrees=45
  twistMotion=Free
 }
 {
  bone1Name=thigh_twist_01_l
  bone2Name=pelvis
  swing1LimitDegrees=45
  swing1Motion=Free
  swing2LimitDegrees=45
  swing2Motion=Free
  twistLimitDegrees=45
  twistMotion=Free
 }
 {
  bone1Name=calf_r
  bone2Name=pelvis
  swing1LimitDegrees=45
  swing1Motion=Free
  swing2LimitDegrees=45
  swing2Motion=Free
  twistLimitDegrees=45
  twistMotion=Free
 }
 {
  bone1Name=thigh_twist_01_r
  bone2Name=pelvis
  swing1LimitDegrees=45
  swing1Motion=Free
  swing2LimitDegrees=45
  swing2Motion=Free
  twistLimitDegrees=45
  twistMotion=Free
 }
 {
  bone1Name=dyn_coat_base_mid_l
  bone2Name=pelvis
  swing1LimitDegrees=45
  swing1Motion=Free
  swing2LimitDegrees=45
  swing2Motion=Free
  twistLimitDegrees=45
  twistMotion=Free
 }
 {
  bone1Name=dyn_coat_base_bk_l
  bone2Name=pelvis
  swing1LimitDegrees=45
  swing1Motion=Free
  swing2LimitDegrees=45
  swing2Motion=Free
  twistLimitDegrees=45
  twistMotion=Free
 }
 {
  bone1Name=dyn_coat_base_bk_r
  bone2Name=pelvis
  swing1LimitDegrees=45
  swing1Motion=Free
  swing2LimitDegrees=45
  swing2Motion=Free
  twistLimitDegrees=45
  twistMotion=Free
 }
 {
  bone1Name=dyn_coat_base_frnt_r
  bone2Name=pelvis
  swing1LimitDegrees=45
  swing1Motion=Free
  swing2LimitDegrees=45
  swing2Motion=Free
  twistLimitDegrees=45
  twistMotion=Free
 }
 {
  bone1Name=dyn_coat_base_frnt_l
  bone2Name=pelvis
  swing1LimitDegrees=45
  swing1Motion=Free
  swing2LimitDegrees=45
  swing2Motion=Free
  twistLimitDegrees=45
  twistMotion=Free
 }
 {
  bone1Name=dyn_coat_02_frnt_l
  bone2Name=dyn_coat_02_mid_l
  swing1LimitDegrees=45
  swing1Motion=Free
  swing2LimitDegrees=45
  swing2Motion=Free
  twistLimitDegrees=45
  twistMotion=Free
 }
 {
  bone1Name=spine_01
  bone2Name=pelvis
  swing1LimitDegrees=45
  swing1Motion=Free
  swing2LimitDegrees=45
  swing2Motion=Free
  twistLimitDegrees=45
  twistMotion=Free
 }
 {
  bone1Name=spine_02
  bone2Name=spine_01
  swing1LimitDegrees=45
  swing1Motion=Free
  swing2LimitDegrees=45
  swing2Motion=Free
  twistLimitDegrees=45
  twistMotion=Free
 }
 {
  bone1Name=spine_03
  bone2Name=spine_02
  swing1LimitDegrees=45
  swing1Motion=Free
  swing2LimitDegrees=45
  swing2Motion=Free
  twistLimitDegrees=45
  twistMotion=Free
 }
 {
  bone1Name=spine_04
  bone2Name=spine_03
  swing1LimitDegrees=45
  swing1Motion=Free
  swing2LimitDegrees=45
  swing2Motion=Free
  twistLimitDegrees=45
  twistMotion=Free
 }
 {
  bone1Name=spine_05
  bone2Name=spine_04
  swing1LimitDegrees=45
  swing1Motion=Free
  swing2LimitDegrees=45
  swing2Motion=Free
  twistLimitDegrees=45
  twistMotion=Free
 }
 {
  bone1Name=upperarm_l
  bone2Name=spine_05
  swing1LimitDegrees=45
  swing1Motion=Free
  swing2LimitDegrees=45
  swing2Motion=Free
  twistLimitDegrees=45
  twistMotion=Free
 }
 {
  bone1Name=lowerarm_l
  bone2Name=upperarm_l
  swing1LimitDegrees=45
  swing1Motion=Free
  swing2LimitDegrees=45
  swing2Motion=Free
  twistLimitDegrees=45
  twistMotion=Free
 }
 {
  bone1Name=upperarm_r
  bone2Name=spine_05
  swing1LimitDegrees=45
  swing1Motion=Free
  swing2LimitDegrees=45
  swing2Motion=Free
  twistLimitDegrees=45
  twistMotion=Free
 }
 {
  bone1Name=lowerarm_r
  bone2Name=upperarm_r
  swing1LimitDegrees=45
  swing1Motion=Free
  swing2LimitDegrees=45
  swing2Motion=Free
  twistLimitDegrees=45
  twistMotion=Free
 }
 {
  bone1Name=pelvis
  bone2Name=root
  swing1LimitDegrees=45
  swing1Motion=Free
  swing2LimitDegrees=45
  swing2Motion=Free
  twistLimitDegrees=45
  twistMotion=Free
 }
 {
  bone1Name=extra_ground
  bone2Name=root
  swing1LimitDegrees=45
  swing1Motion=Free
  swing2LimitDegrees=45
  swing2Motion=Free
  twistLimitDegrees=45
  twistMotion=Free
 }
 {
  bone1Name=dyn_coat_loop_base
  bone2Name=pelvis
  swing1LimitDegrees=45
  swing1Motion=Free
  swing2LimitDegrees=45
  swing2Motion=Free
  twistLimitDegrees=45
  twistMotion=Free
 }
 {bone1Name=dyn_coat_loop_parent, bone2Name=dyn_coat_loop_base, swing1LimitDegrees=15, swing2LimitDegrees=5, twistLimitDegrees=5}
]
