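# Physics asset for the Bountress female coat mesh, named as the collider set for her cape.
# Capsules and spheres cover the body, while capsules and boxes cover coat, scarf and arm wire bones.
# Only the left front and mid coat chains and wires have limited angular constraints; all other joints rotate freely.
# The root carries two spheres, including a very large one offset downward, apparently acting as a ground collider.
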PhysicsAsset /BRCosmetics/Characters/Player/Female/Medium/Bodies/F_MED_Bountress/Meshes/F_MED_Bountress_Coat_Physics_CapeColliders
bodies {
 calf_l [{shapeType=Capsule, length=34, radius=6, shapeName=calf_l_capsule, center=<-19.4567 -2.431 -0.5072>, rotation=<rot 90 0 0>}]
 calf_r [{shapeType=Capsule, length=38.9082, radius=6.0918, shapeName=calf_r_capsule, center=<25.2816 2.6863 0.6944>, rotation=<rot 90 0 0>}]
 dyn_coat_01_bk_l [{
   shapeType=Capsule
   length=6.1739
   radius=11.4999
   shapeName=dyn_coat_01_bk_l_capsule
   center=<12.647 1.9508 7.2322>
   rotation=<rot 6.9822 -2.3845 -69.6103>
  }]
 dyn_coat_01_frnt_l [{
   shapeType=Box
   extentX=14.4213
   extentY=17.6337
   extentZ=4.4016
   shapeName=dyn_coat_01_frnt_l_box
   center=<7.2984 0 0>
   rotation=<rot -0.0001 -0.0002 -9.7845>
  }]
 dyn_coat_01_frnt_r [{
   shapeType=Capsule
   length=12.5408
   radius=7.7992
   shapeName=dyn_coat_1_bk_r_capsule
   center=<19.5562 -0.8262 4.1507>
   rotation=<rot 14.0241 156.9639 -88.4837>
  }]
 dyn_coat_01_mid_l [{
   shapeType=Box
   extentX=12.2976
   extentY=17.6621
   extentZ=3.3675
   shapeName=dyn_coat_01_mid_l_box
   center=<6.9427 0 0>
   rotation=<rot 0 -0.0001 11.7822>
  }]
 dyn_coat_02_frnt_l [{
   shapeType=Box
   extentX=20.1469
   extentY=30.6904
   extentZ=5.2996
   shapeName=dyn_coat_02_frnt_l_box
   center=<13.7392 0.2906 0.4488>
   rotation=<rot 5.0034 40.2694 -4.5614>
  }]
 dyn_coat_02_frnt_r [{
   shapeType=Capsule
   length=9.7547
   radius=10.5854
   shapeName=dyn_coat_1_bk_r_capsule
   center=<18.3154 4.1188 4.8437>
   rotation=<rot 14.2014 154.3567 -79.4059>
  }]
 dyn_coat_02_mid_l [{shapeType=Box, extentX=13.4707, extentY=13.4326, extentZ=3.3675, shapeName=dyn_coat_02_mid_l_box, center=<8.0175 0 0>}]
 dyn_coat_1_bk_r [{
   shapeType=Capsule
   length=9.8081
   radius=10.5319
   shapeName=dyn_coat_1_bk_r_capsule
   center=<16.2917 1.1454 6.0719>
   rotation=<rot 12.0092 162.345 -68.3861>
  }]
 dyn_coat_base_frnt_l [{shapeType=Sphere, radius=0.5, shapeName=dyn_coat_base_frnt_l_sphere}]
 dyn_coat_base_mid_l [{shapeType=Sphere, radius=0.5, shapeName=dyn_coat_base_mid_l_sphere}]
 dyn_coat_parent_bk_l1 [{
   shapeType=Capsule
   length=4.8904
   radius=11.9067
   shapeName=dyn_coat_parent_bk_l1_capsule
   center=<9.4781 0.5755 2.7385>
   rotation=<rot 2.8255 0.9188 -58.6097>
  }]
 dyn_coat_parent_bk_r [{
   shapeType=Capsule
   length=9.8081
   radius=10.5319
   shapeName=dyn_coat_1_bk_r_capsule
   center=<9.969 -2.2745 6.7728>
   rotation=<rot 23.9031 85.6746 -95.9295>
  }]
 dyn_coat_parent_frnt_l1 [{
   shapeType=Box
   extentX=23.1066
   extentY=17.5817
   extentZ=6.2312
   shapeName=dyn_coat_parent_frnt_l1_box
   center=<11.9619 0 0>
   rotation=<rot -0.0002 -0.0002 -3.3559>
  }]
 dyn_coat_parent_frnt_r [{
   shapeType=Capsule
   length=11.1011
   radius=9.2389
   shapeName=dyn_coat_1_bk_r_capsule
   center=<9.2885 -1.2026 5.6777>
   rotation=<rot 2.365 -175.1278 -85.8281>
  }]
 dyn_coat_parent_mid_l [{
   shapeType=Box
   extentX=11.7143
   extentY=17.879
   extentZ=3.3675
   shapeName=dyn_coat_parent_mid_l_box
   center=<6.0988 0 0>
   rotation=<rot 0 -0.0001 11.8487>
  }]
 dyn_scarf_01 [{
   shapeType=Capsule
   length=28.3467
   radius=5.5661
   shapeName=dyn_scarf_01_capsule
   center=<16.0747 -1.218 1.3446>
   rotation=<rot 0 89.3096 -86.7956>
  }]
 dyn_scarf_02 [{
   shapeType=Capsule
   length=25.1559
   radius=6.3886
   shapeName=dyn_scarf_02_capsule
   center=<19.8391 -0.5628 2.1036>
   rotation=<rot 0.584 90.3481 -84.0974>
  }]
 dyn_scarf_base [{shapeType=Capsule, length=0.505, radius=0.505, shapeName=dyn_scarf_base_capsule, rotation=<rot 0 -90 0>}]
 dyn_scarf_parent [{
   shapeType=Capsule
   length=35.0141
   radius=5.9152
   shapeName=dyn_scarf_parent_capsule
   center=<21.7189 0 1.5968>
   rotation=<rot 87.5061 180 180>
  }]
 dyn_wire_1 [{shapeType=Capsule, length=2.4418, radius=1.9966, shapeName=dyn_wire_1_capsule, center=<0 -3.0616 -0.0693>, rotation=<rot 86.5614 90 -180>}]
 dyn_wire_2 [{shapeType=Capsule, length=2.4418, radius=1.9966, shapeName=dyn_wire_1_capsule, center=<0 -3.0616 -0.0693>, rotation=<rot 86.5614 90 -180>}]
 foot_l [{shapeType=Capsule, length=26, radius=7.5, shapeName=foot_r_capsule, center=<-4.3789 2.9446 0>, rotation=<rot 0 0 90>}]
 foot_r [{shapeType=Capsule, length=26, radius=7.5, shapeName=foot_r_capsule, center=<3.3697 -5.0998 0>, rotation=<rot 0 0 90>}]
 hand_l [{
   shapeType=Capsule
   length=7.1594
   radius=6.5021
   shapeName=hand_r_capsule
   center=<10.042 -2.116 -0.5585>
   rotation=<rot 1.2637 -90 -92.3753>
  }]
 hand_r [{shapeType=Capsule, length=7.1594, radius=6.5021, shapeName=hand_r_capsule, center=<-9.6877 0.73 -0.4975>, rotation=<rot 1.2637 -90 -92.3753>}]
 head [{shapeType=Sphere, radius=17.3125, center=<11.028 -0.174 0>}]
 lowerarm_l [{
   shapeType=Capsule
   length=18.4966
   radius=5.1795
   shapeName=lowerarm_l_capsule
   center=<11.8221 -0.908 -0.3383>
   rotation=<rot 0 -90 -88.361>
  }]
 lowerarm_r [{
   shapeType=Capsule
   length=18.4966
   radius=5.1795
   shapeName=lowerarm_l_capsule
   center=<-12.3205 -0.908 0.3525>
   rotation=<rot 0 -90 -88.361>
  }]
 pelvis [{shapeType=Capsule, length=5.069, radius=19.6202, shapeName=pelvis_capsule, center=<0 -5.6419 0>}]
 root [
  {shapeType=Sphere, radius=30.13, shapeName=root_sphere, center=<0.0002 -7.2623 8.5972>}
  {shapeType=Sphere, radius=175, shapeName=root_sphere, center=<0 -8.6403 -165.6077>}
 ]
 spine_01 [{shapeType=Capsule, length=0.505, radius=0.505, shapeName=spine_01_capsule, rotation=<rot 0 -90 0>}]
 spine_02 [{
   shapeType=Capsule
   length=10.9001
   radius=16.5252
   shapeName=spine_02_capsule
   center=<1.7406 -1.8622 -0.9152>
   rotation=<rot -1.0658 -169.4781 -156.8654>
  }]
 spine_03 [{shapeType=Capsule, length=0.505, radius=0.505, shapeName=spine_03_capsule, center=<1.1754 -7.1962 0.0007>, rotation=<rot 0.0091 -90 -0.0408>}]
 spine_04 [{
   shapeType=Capsule
   length=5.5492
   radius=17.5651
   shapeName=spine_04_capsule
   center=<0.2912 0.2159 -0.1196>
   rotation=<rot -12.1921 -90.0001 2.2691>
  }]
 spine_05 [{shapeType=Capsule, length=9.96, radius=9.9054, shapeName=spine_05_capsule}]
 thigh_r [{
   shapeType=Capsule
   length=10.5429
   radius=6.1545
   shapeName=thigh_twist_01_r_capsule
   center=<5.7201 -1.4509 12.2551>
   rotation=<rot 79.1159 0 0>
  }]
 thigh_twist_01_l [{
   shapeType=Capsule
   length=23.0917
   radius=10.9083
   shapeName=thigh_twist_01_l_capsule
   center=<4.7138 -0.8436 0.8044>
   rotation=<rot 85 179.9996 179.9999>
  }]
 thigh_twist_01_r []
 upperarm_l [
  {
   shapeType=Capsule
   length=13.4135
   radius=5.8763
   shapeName=upperarm_l_capsule
   center=<15.7241 -0.9761 -0.2202>
   rotation=<rot 3.4744 -87.5326 -90.0942>
  }
  {
   shapeType=Capsule
   length=13.4135
   radius=5.8763
   shapeName=upperarm_l_capsule
   center=<15.5218 -5.7098 0.0232>
   rotation=<rot 3.4744 -87.5326 -90.0942>
  }
 ]
 upperarm_r [{
   shapeType=Capsule
   length=22.4586
   radius=5.734
   shapeName=upperarm_l_capsule
   center=<-10.0218 1.8218 0.5701>
   rotation=<rot 3.4744 -87.5326 -90.0942>
  }]
}
constraints [
 {
  bone1Name=foot_l
  bone2Name=calf_l
  swing1LimitDegrees=45
  swing1Motion=Free
  swing2LimitDegrees=45
  swing2Motion=Free
  twistLimitDegrees=45
  twistMotion=Free
 }
 {
  bone1Name=foot_r
  bone2Name=calf_r
  swing1LimitDegrees=45
  swing1Motion=Free
  swing2LimitDegrees=45
  swing2Motion=Free
  twistLimitDegrees=45
  twistMotion=Free
 }
 {
  bone1Name=calf_l
  bone2Name=pelvis
  swing1LimitDegrees=45
  swing1Motion=Free
  swing2LimitDegrees=45
  swing2Motion=Free
  twistLimitDegrees=45
  twistMotion=Free
 }
 {
  bone1Name=thigh_twist_01_l
  bone2Name=pelvis
  swing1LimitDegrees=45
  swing1Motion=Free
  swing2LimitDegrees=45
  swing2Motion=Free
  twistLimitDegrees=45
  twistMotion=Free
 }
 {
  bone1Name=calf_r
  bone2Name=thigh_r
  swing1LimitDegrees=45
  swing1Motion=Free
  swing2LimitDegrees=45
  swing2Motion=Free
  twistLimitDegrees=45
  twistMotion=Free
 }
 {
  bone1Name=thigh_twist_01_r
  bone2Name=thigh_r
  swing1LimitDegrees=45
  swing1Motion=Free
  swing2LimitDegrees=45
  swing2Motion=Free
  twistLimitDegrees=45
  twistMotion=Free
 }
 {
  bone1Name=spine_01
  bone2Name=pelvis
  swing1LimitDegrees=45
  swing1Motion=Free
  swing2LimitDegrees=45
  swing2Motion=Free
  twistLimitDegrees=45
  twistMotion=Free
 }
 {
  bone1Name=spine_02
  bone2Name=spine_01
  swing1LimitDegrees=45
  swing1Motion=Free
  swing2LimitDegrees=45
  swing2Motion=Free
  twistLimitDegrees=45
  twistMotion=Free
 }
 {
  bone1Name=spine_03
  bone2Name=spine_02
  swing1LimitDegrees=45
  swing1Motion=Free
  swing2LimitDegrees=45
  swing2Motion=Free
  twistLimitDegrees=45
  twistMotion=Free
 }
 {
  bone1Name=spine_04
  bone2Name=spine_03
  swing1LimitDegrees=45
  swing1Motion=Free
  swing2LimitDegrees=45
  swing2Motion=Free
  twistLimitDegrees=45
  twistMotion=Free
 }
 {
  bone1Name=spine_05
  bone2Name=spine_04
  swing1LimitDegrees=45
  swing1Motion=Free
  swing2LimitDegrees=45
  swing2Motion=Free
  twistLimitDegrees=45
  twistMotion=Free
 }
 {
  bone1Name=upperarm_l
  bone2Name=spine_05
  swing1LimitDegrees=45
  swing1Motion=Free
  swing2LimitDegrees=45
  swing2Motion=Free
  twistLimitDegrees=45
  twistMotion=Free
 }
 {
  bone1Name=lowerarm_l
  bone2Name=upperarm_l
  swing1LimitDegrees=45
  swing1Motion=Free
  swing2LimitDegrees=45
  swing2Motion=Free
  twistLimitDegrees=45
  twistMotion=Free
 }
 {
  bone1Name=upperarm_r
  bone2Name=spine_05
  swing1LimitDegrees=45
  swing1Motion=Free
  swing2LimitDegrees=45
  swing2Motion=Free
  twistLimitDegrees=45
  twistMotion=Free
 }
 {
  bone1Name=lowerarm_r
  bone2Name=upperarm_r
  swing1LimitDegrees=45
  swing1Motion=Free
  swing2LimitDegrees=45
  swing2Motion=Free
  twistLimitDegrees=45
  twistMotion=Free
 }
 {
  bone1Name=dyn_scarf_base
  bone2Name=spine_05
  swing1LimitDegrees=45
  swing1Motion=Free
  swing2LimitDegrees=45
  swing2Motion=Free
  twistLimitDegrees=45
  twistMotion=Free
 }
 {
  bone1Name=dyn_coat_base_frnt_l
  bone2Name=pelvis
  swing1LimitDegrees=45
  swing1Motion=Free
  swing2LimitDegrees=45
  swing2Motion=Free
  twistLimitDegrees=45
  twistMotion=Free
 }
 {bone1Name=dyn_coat_parent_frnt_l1, bone2Name=dyn_coat_base_frnt_l, swing1LimitDegrees=70, swing2LimitDegrees=70, twistLimitDegrees=20}
 {bone1Name=dyn_coat_01_frnt_l, bone2Name=dyn_coat_parent_frnt_l1, swing1LimitDegrees=20, swing2LimitDegrees=20, twistLimitDegrees=5}
 {bone1Name=dyn_coat_02_frnt_l, bone2Name=dyn_coat_01_frnt_l, swing1LimitDegrees=20, swing2LimitDegrees=20, twistLimitDegrees=5}
 {
  bone1Name=dyn_coat_base_mid_l
  bone2Name=pelvis
  swing1LimitDegrees=45
  swing1Motion=Free
  swing2LimitDegrees=45
  swing2Motion=Free
  twistLimitDegrees=45
  twistMotion=Free
 }
 {bone1Name=dyn_coat_parent_mid_l, bone2Name=dyn_coat_base_mid_l, swing1LimitDegrees=20, swing2LimitDegrees=20, twistLimitDegrees=3}
 {bone1Name=dyn_coat_01_mid_l, bone2Name=dyn_coat_parent_mid_l, swing1LimitDegrees=20, swing2LimitDegrees=20, twistLimitDegrees=3}
 {bone1Name=dyn_coat_02_mid_l, bone2Name=dyn_coat_01_mid_l, swing1LimitDegrees=20, swing2LimitDegrees=20, twistLimitDegrees=3}
 {
  bone1Name=pelvis
  bone2Name=root
  swing1LimitDegrees=45
  swing1Motion=Free
  swing2LimitDegrees=45
  swing2Motion=Free
  twistLimitDegrees=45
  twistMotion=Free
 }
 {bone1Name=dyn_wire_1, bone2Name=upperarm_l, swing1LimitDegrees=5, swing2LimitDegrees=60, twistLimitDegrees=4}
 {bone1Name=dyn_wire_2, bone2Name=upperarm_l, swing1LimitDegrees=5, swing2LimitDegrees=60, twistLimitDegrees=5}
 {
  bone1Name=dyn_coat_01_bk_l
  bone2Name=dyn_coat_parent_bk_l1
  swing1LimitDegrees=45
  swing1Motion=Free
  swing2LimitDegrees=45
  swing2Motion=Free
  twistLimitDegrees=45
  twistMotion=Free
 }
 {
  bone1Name=dyn_coat_parent_bk_r
  bone2Name=pelvis
  swing1LimitDegrees=45
  swing1Motion=Free
  swing2LimitDegrees=45
  swing2Motion=Free
  twistLimitDegrees=45
  twistMotion=Free
 }
 {
  bone1Name=dyn_coat_1_bk_r
  bone2Name=dyn_coat_parent_bk_r
  swing1LimitDegrees=45
  swing1Motion=Free
  swing2LimitDegrees=45
  swing2Motion=Free
  twistLimitDegrees=45
  twistMotion=Free
 }
 {
  bone1Name=dyn_coat_parent_frnt_r
  bone2Name=pelvis
  swing1LimitDegrees=45
  swing1Motion=Free
  swing2LimitDegrees=45
  swing2Motion=Free
  twistLimitDegrees=45
  twistMotion=Free
 }
 {
  bone1Name=dyn_coat_01_frnt_r
  bone2Name=dyn_coat_parent_frnt_r
  swing1LimitDegrees=45
  swing1Motion=Free
  swing2LimitDegrees=45
  swing2Motion=Free
  twistLimitDegrees=45
  twistMotion=Free
 }
 {
  bone1Name=dyn_coat_02_frnt_r
  bone2Name=dyn_coat_01_frnt_r
  swing1LimitDegrees=45
  swing1Motion=Free
  swing2LimitDegrees=45
  swing2Motion=Free
  twistLimitDegrees=45
  twistMotion=Free
 }
 {
  bone1Name=hand_r
  bone2Name=lowerarm_r
  swing1LimitDegrees=45
  swing1Motion=Free
  swing2LimitDegrees=45
  swing2Motion=Free
  twistLimitDegrees=45
  twistMotion=Free
 }
 {
  bone1Name=hand_l
  bone2Name=lowerarm_l
  swing1LimitDegrees=45
  swing1Motion=Free
  swing2LimitDegrees=45
  swing2Motion=Free
  twistLimitDegrees=45
  twistMotion=Free
 }
 {
  bone1Name=thigh_r
  bone2Name=pelvis
  swing1LimitDegrees=45
  swing1Motion=Free
  swing2LimitDegrees=45
  swing2Motion=Free
  twistLimitDegrees=45
  twistMotion=Free
 }
 {
  bone1Name=dyn_coat_parent_bk_l1
  bone2Name=pelvis
  swing1LimitDegrees=45
  swing1Motion=Free
  swing2LimitDegrees=45
  swing2Motion=Free
  twistLimitDegrees=45
  twistMotion=Free
 }
 {
  bone1Name=dyn_scarf_parent
  bone2Name=dyn_scarf_base
  swing1LimitDegrees=45
  swing1Motion=Free
  swing2LimitDegrees=45
  swing2Motion=Free
  twistLimitDegrees=45
  twistMotion=Free
 }
 {
  bone1Name=dyn_scarf_01
  bone2Name=dyn_scarf_parent
  swing1LimitDegrees=45
  swing1Motion=Free
  swing2LimitDegrees=45
  swing2Motion=Free
  twistLimitDegrees=45
  twistMotion=Free
 }
 {
  bone1Name=dyn_scarf_02
  bone2Name=dyn_scarf_01
  swing1LimitDegrees=45
  swing1Motion=Free
  swing2LimitDegrees=45
  swing2Motion=Free
  twistLimitDegrees=45
  twistMotion=Free
 }
 {
  bone1Name=head
  bone2Name=spine_05
  swing1LimitDegrees=45
  swing1Motion=Free
  swing2LimitDegrees=45
  swing2Motion=Free
  twistLimitDegrees=45
  twistMotion=Free
 }
]
