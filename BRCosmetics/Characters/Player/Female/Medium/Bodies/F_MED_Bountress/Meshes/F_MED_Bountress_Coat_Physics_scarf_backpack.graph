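# Physics asset for the Bountress female medium character's secondary motion on left-side coat panels, scarf and two small wire bones.
# Primitive colliders cover pelvis, spine, arms, legs and feet, plus two root spheres; thigh_twist_01_r has no shape.
# Angle-limited chains hang from freely rotating base bones on pelvis (front and mid coat) and spine_05 (scarf).
# Wires are limited off upperarm_l, while the back coat body and skeleton joints rotate freely.
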PhysicsAsset /BRCosmetics/Characters/Player/Female/Medium/Bodies/F_MED_Bountress/Meshes/F_MED_Bountress_Coat_Physics_scarf_backpack
bodies {
 calf_l [{shapeType=Capsule, length=34, radius=6, shapeName=calf_l_capsule, center=<-19.4567 -2.431 -0.5072>, rotation=<rot 90 0 0>}]
 calf_r [{shapeType=Capsule, length=38.9082, radius=6.0918, shapeName=calf_r_capsule, center=<25.2816 2.6863 0.6944>, rotation=<rot 90 0 0>}]
 dyn_coat_01_bk_l [{
   shapeType=Capsule
   length=1.9529
   radius=9.5626
   shapeName=dyn_coat_01_bk_l_capsule
   center=<12.3324 3.0678 10.1776>
   rotation=<rot 4.489 4.2315 -68.9482>
  }]
 dyn_coat_01_frnt_l [{
   shapeType=Capsule
   length=6.3511
   radius=11.1932
   shapeName=dyn_coat_01_frnt_l_capsule
   center=<10.5113 0 12.3228>
   rotation=<rot 90 -160.5288 -160.5287>
  }]
 dyn_coat_01_mid_l [{
   shapeType=Capsule
   length=8.3126
   radius=8.5397
   shapeName=dyn_coat_01_mid_l_capsule
   center=<8.9422 -0.0016 10.08>
   rotation=<rot 88.226 -179.9802 -179.9708>
  }]
 dyn_coat_02_frnt_l [{
   shapeType=Capsule
   length=11.8805
   radius=9.0045
   shapeName=dyn_coat_02_frnt_l_capsule
   center=<12.9389 0 11.3968>
   rotation=<rot 88.174 -179.9998 -179.9998>
  }]
 dyn_coat_02_mid_l [{
   shapeType=Capsule
   length=5.0359
   radius=9.7211
   shapeName=dyn_coat_02_mid_l_capsule
   center=<7.6952 0 10.4441>
   rotation=<rot 88.3934 -179.9999 -179.9999>
  }]
 dyn_coat_base_frnt_l [{shapeType=Sphere, radius=0.5, shapeName=dyn_coat_base_frnt_l_sphere}]
 dyn_coat_base_mid_l [{shapeType=Sphere, radius=0.5, shapeName=dyn_coat_base_mid_l_sphere}]
 dyn_coat_parent_frnt_l1 [{
   shapeType=Capsule
   length=9.3073
   radius=10.2091
   shapeName=dyn_coat_parent_frnt_l1_capsule
   center=<19.2982 0 11.418>
   rotation=<rot 87.9572 -0.0001 -0.0001>
  }]
 dyn_scarf_01 [{
   shapeType=Capsule
   length=17.1449
   radius=5.5661
   shapeName=dyn_scarf_01_capsule
   center=<14.932 -1.2042 1.2806>
   rotation=<rot 0 89.3096 -86.7956>
  }]
 dyn_scarf_02 [{
   shapeType=Capsule
   length=22.4556
   radius=6.3886
   shapeName=dyn_scarf_02_capsule
   center=<19.8391 -0.5628 2.1036>
   rotation=<rot 0.584 90.3481 -84.0974>
  }]
 dyn_scarf_base [{shapeType=Capsule, length=0.505, radius=0.505, shapeName=dyn_scarf_base_capsule, rotation=<rot 0 -90 0>}]
 dyn_scarf_parent [{
   shapeType=Box
   extentX=31.0826
   extentY=9.3583
   extentZ=4.8339
   shapeName=dyn_scarf_parent_box
   center=<18.2544 0 -0.2138>
   rotation=<rot 3.2663 0 0>
  }]
 dyn_wire_1 [{shapeType=Capsule, length=2.4418, radius=1.9966, shapeName=dyn_wire_1_capsule, center=<0 -3.0616 -0.0693>, rotation=<rot 86.5614 90 -180>}]
 dyn_wire_2 [{shapeType=Capsule, length=2.4418, radius=1.9966, shapeName=dyn_wire_1_capsule, center=<0 -3.0616 -0.0693>, rotation=<rot 86.5614 90 -180>}]
 foot_l [{shapeType=Capsule, length=26, radius=7.5, shapeName=foot_r_capsule, center=<-4.3789 2.9446 0>, rotation=<rot 0 0 90>}]
 foot_r [{shapeType=Capsule, length=26, radius=7.5, shapeName=foot_r_capsule, center=<3.3697 -5.0998 0>, rotation=<rot 0 0 90>}]
 lowerarm_l [{
   shapeType=Capsule
   length=18.4966
   radius=5.1795
   shapeName=lowerarm_l_capsule
   center=<11.8221 -0.908 -0.3383>
   rotation=<rot 0 -90 -88.361>
  }]
 lowerarm_r [{
   shapeType=Capsule
   length=18.4966
   radius=5.1795
   shapeName=lowerarm_l_capsule
   center=<-12.3205 -0.908 0.3525>
   rotation=<rot 0 -90 -88.361>
  }]
 pelvis [{shapeType=Capsule, length=4.108, radius=12.5885, shapeName=pelvis_capsule, center=<0 -1.4312 0>}]
 root [
  {shapeType=Sphere, radius=30.13, shapeName=root_sphere, center=<0.0002 -7.2623 -0.905>}
  {shapeType=Sphere, radius=175, shapeName=root_sphere, center=<0 -8.6403 -165.6077>}
 ]
 spine_01 [{shapeType=Capsule, length=0.505, radius=0.505, shapeName=spine_01_capsule, rotation=<rot 0 -90 0>}]
 spine_02 [{
   shapeType=Capsule
   length=3.5886
   radius=10.2932
   shapeName=spine_02_capsule
   center=<1.3313 0.2931 -0.4397>
   rotation=<rot 0 -169.0228 -156.8696>
  }]
 spine_03 [{shapeType=Capsule, length=0.505, radius=0.505, shapeName=spine_03_capsule, center=<1.1754 -7.1962 0.0007>, rotation=<rot 0.0091 -90 -0.0408>}]
 spine_04 [{
   shapeType=Capsule
   length=4.167
   radius=11.7862
   shapeName=spine_04_capsule
   center=<0.2912 1.564 0.1717>
   rotation=<rot -12.1921 -90 -1.484>
  }]
 spine_05 [{shapeType=Capsule, length=7.4783, radius=9.9054, shapeName=spine_05_capsule}]
 thigh_twist_01_l [{
   shapeType=Capsule
   length=30.9821
   radius=5.9968
   shapeName=thigh_twist_01_l_capsule
   center=<9.1659 -0.8718 -0.1151>
   rotation=<rot -26.4984 -90.0001 -90.3904>
  }]
 thigh_twist_01_r []
 upperarm_l [{
   shapeType=Capsule
   length=13.4135
   radius=5.8763
   shapeName=upperarm_l_capsule
   center=<15.7241 -0.9761 -0.2202>
   rotation=<rot 3.4744 -87.5326 -90.0942>
  }]
 upperarm_r [{
   shapeType=Capsule
   length=22.4586
   radius=5.734
   shapeName=upperarm_l_capsule
   center=<-10.0218 1.8218 0.5701>
   rotation=<rot 3.4744 -87.5326 -90.0942>
  }]
}
constraints [
 {
  bone1Name=foot_l
  bone2Name=calf_l
  swing1LimitDegrees=45
  swing1Motion=Free
  swing2LimitDegrees=45
  swing2Motion=Free
  twistLimitDegrees=45
  twistMotion=Free
 }
 {
  bone1Name=foot_r
  bone2Name=calf_r
  swing1LimitDegrees=45
  swing1Motion=Free
  swing2LimitDegrees=45
  swing2Motion=Free
  twistLimitDegrees=45
  twistMotion=Free
 }
 {
  bone1Name=calf_l
  bone2Name=pelvis
  swing1LimitDegrees=45
  swing1Motion=Free
  swing2LimitDegrees=45
  swing2Motion=Free
  twistLimitDegrees=45
  twistMotion=Free
 }
 {
  bone1Name=thigh_twist_01_l
  bone2Name=pelvis
  swing1LimitDegrees=45
  swing1Motion=Free
  swing2LimitDegrees=45
  swing2Motion=Free
  twistLimitDegrees=45
  twistMotion=Free
 }
 {
  bone1Name=calf_r
  bone2Name=pelvis
  swing1LimitDegrees=45
  swing1Motion=Free
  swing2LimitDegrees=45
  swing2Motion=Free
  twistLimitDegrees=45
  twistMotion=Free
 }
 {
  bone1Name=thigh_twist_01_r
  bone2Name=pelvis
  swing1LimitDegrees=45
  swing1Motion=Free
  swing2LimitDegrees=45
  swing2Motion=Free
  twistLimitDegrees=45
  twistMotion=Free
 }
 {
  bone1Name=spine_01
  bone2Name=pelvis
  swing1LimitDegrees=45
  swing1Motion=Free
  swing2LimitDegrees=45
  swing2Motion=Free
  twistLimitDegrees=45
  twistMotion=Free
 }
 {
  bone1Name=spine_02
  bone2Name=spine_01
  swing1LimitDegrees=45
  swing1Motion=Free
  swing2LimitDegrees=45
  swing2Motion=Free
  twistLimitDegrees=45
  twistMotion=Free
 }
 {
  bone1Name=spine_03
  bone2Name=spine_02
  swing1LimitDegrees=45
  swing1Motion=Free
  swing2LimitDegrees=45
  swing2Motion=Free
  twistLimitDegrees=45
  twistMotion=Free
 }
 {
  bone1Name=spine_04
  bone2Name=spine_03
  swing1LimitDegrees=45
  swing1Motion=Free
  swing2LimitDegrees=45
  swing2Motion=Free
  twistLimitDegrees=45
  twistMotion=Free
 }
 {
  bone1Name=spine_05
  bone2Name=spine_04
  swing1LimitDegrees=45
  swing1Motion=Free
  swing2LimitDegrees=45
  swing2Motion=Free
  twistLimitDegrees=45
  twistMotion=Free
 }
 {
  bone1Name=upperarm_l
  bone2Name=spine_05
  swing1LimitDegrees=45
  swing1Motion=Free
  swing2LimitDegrees=45
  swing2Motion=Free
  twistLimitDegrees=45
  twistMotion=Free
 }
 {
  bone1Name=lowerarm_l
  bone2Name=upperarm_l
  swing1LimitDegrees=45
  swing1Motion=Free
  swing2LimitDegrees=45
  swing2Motion=Free
  twistLimitDegrees=45
  twistMotion=Free
 }
 {
  bone1Name=upperarm_r
  bone2Name=spine_05
  swing1LimitDegrees=45
  swing1Motion=Free
  swing2LimitDegrees=45
  swing2Motion=Free
  twistLimitDegrees=45
  twistMotion=Free
 }
 {
  bone1Name=lowerarm_r
  bone2Name=upperarm_r
  swing1LimitDegrees=45
  swing1Motion=Free
  swing2LimitDegrees=45
  swing2Motion=Free
  twistLimitDegrees=45
  twistMotion=Free
 }
 {
  bone1Name=dyn_scarf_base
  bone2Name=spine_05
  swing1LimitDegrees=45
  swing1Motion=Free
  swing2LimitDegrees=45
  swing2Motion=Free
  twistLimitDegrees=45
  twistMotion=Free
 }
 {bone1Name=dyn_scarf_parent, bone2Name=dyn_scarf_base, swing1LimitDegrees=30, swing2LimitDegrees=10, twistLimitDegrees=10}
 {bone1Name=dyn_scarf_01, bone2Name=dyn_scarf_parent, swing1LimitDegrees=60, swing2LimitDegrees=60, twistLimitDegrees=5}
 {bone1Name=dyn_scarf_02, bone2Name=dyn_scarf_01, swing1LimitDegrees=60, swing2LimitDegrees=60, twistLimitDegrees=5}
 {
  bone1Name=dyn_coat_base_frnt_l
  bone2Name=pelvis
  swing1LimitDegrees=45
  swing1Motion=Free
  swing2LimitDegrees=45
  swing2Motion=Free
  twistLimitDegrees=45
  twistMotion=Free
 }
 {bone1Name=dyn_coat_parent_frnt_l1, bone2Name=dyn_coat_base_frnt_l, swing1LimitDegrees=70, swing2LimitDegrees=70, twistLimitDegrees=20}
 {bone1Name=dyn_coat_01_frnt_l, bone2Name=dyn_coat_parent_frnt_l1, swing1LimitDegrees=20, swing2LimitDegrees=20, twistLimitDegrees=5}
 {bone1Name=dyn_coat_02_frnt_l, bone2Name=dyn_coat_01_frnt_l, swing1LimitDegrees=20, swing2LimitDegrees=20, twistLimitDegrees=5}
 {
  bone1Name=dyn_coat_base_mid_l
  bone2Name=pelvis
  swing1LimitDegrees=45
  swing1Motion=Free
  swing2LimitDegrees=45
  swing2Motion=Free
  twistLimitDegrees=45
  twistMotion=Free
 }
 {bone1Name=dyn_coat_01_mid_l, bone2Name=dyn_coat_base_mid_l, swing1LimitDegrees=20, swing2LimitDegrees=20, twistLimitDegrees=3}
 {bone1Name=dyn_coat_02_mid_l, bone2Name=dyn_coat_01_mid_l, swing1LimitDegrees=20, swing2LimitDegrees=20, twistLimitDegrees=3}
 {
  bone1Name=pelvis
  bone2Name=root
  swing1LimitDegrees=45
  swing1Motion=Free
  swing2LimitDegrees=45
  swing2Motion=Free
  twistLimitDegrees=45
  twistMotion=Free
 }
 {bone1Name=dyn_wire_1, bone2Name=upperarm_l, swing1LimitDegrees=5, swing2LimitDegrees=60, twistLimitDegrees=4}
 {bone1Name=dyn_wire_2, bone2Name=upperarm_l, swing1LimitDegrees=5, swing2LimitDegrees=60, twistLimitDegrees=5}
 {
  bone1Name=dyn_coat_01_bk_l
  bone2Name=pelvis
  swing1LimitDegrees=45
  swing1Motion=Free
  swing2LimitDegrees=45
  swing2Motion=Free
  twistLimitDegrees=45
  twistMotion=Free
 }
]
