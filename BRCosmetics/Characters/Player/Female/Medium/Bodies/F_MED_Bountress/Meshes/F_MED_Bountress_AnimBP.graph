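# Post-process style AnimBlueprint for the Bountress body that copies the parent mesh pose and adds coat, scarf and spur secondary motion.
# Control Rigs bracket a chain of RigidBody nodes (coat off while sliding, scarf asset chosen by backpack) and spur AnimDynamics.
# Each update converts movement states to rig inputs and rewrites gravity/wind parameters twice, by backpack then airborne state, deriving separate coat and scarf gravity.
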AnimBlueprint /BRCosmetics/Characters/Player/Female/Medium/Bodies/F_MED_Bountress/Meshes/F_MED_Bountress_AnimBP
target_skeleton=&/F_MED_Bountress_Skeleton
graphs [
 AnimGraph {nodes [
   n1 = Control Rig!(
    Source=Copy Pose From Mesh() {bUseAttachedParent=true}
    Alpha
    sliding:double=Get sliding() -> (sliding:double)
    crouching:double=Get f_crouching() -> (f_crouching:double)
    f_crouching_and_moving_2D:double=Get f_crouching_and_moving_2D() -> (f_crouching_and_moving_2D:double)
    rail:bool=Get bIsOnGrindRail()
    not_crouching:double=Get f_not_crouching() -> (f_not_crouching:double)
    RBAN:bool=Get bRigidBodyEnabled()
   ) -> (Pose) {
    alphaBoolBlend {blendOption=Linear}
    controlRigAssetReference {blueprintRigClass=&/F_MED_Bountress_CtrlRig.F_MED_Bountress_CtrlRig_C}
   }
   n2 = RigidBody!(ComponentPose=Local To Component(LocalPose=@n1), bAlphaBoolEnabled:bool=false, OverrideWorldGravity:Vector="0,0,0", ExternalForce) -> (Pose) {
    alphaInputType=Bool
    bClampLinearTranslationLimitToRefPose=true
    bOverrideWorldGravity=true
    overridePhysicsAsset=&/F_MED_Bountress_Coat_Physics_CapeColliders
    simulationSpace=BaseBoneSpace
    alphaBoolBlend {blendInTime=0.2, blendOption=Linear, blendOutTime=0.2}
    baseBoneRef {boneName=pelvis}
    componentAppliedLinearAccClamp=<6000 6000 6000>
    componentLinearAccScale=<0.25 0.25 0.25>
    componentLinearVelScale=<0.25 0.25 0.25>
    overrideWorldGravity {z=-980}
    simSpaceSettings {
     maxAngularAcceleration=500
     maxAngularVelocity=3
     maxLinearAcceleration=400
     maxLinearVelocity=250
     velocityScaleZ=0
     worldAlpha=1
     externalLinearDragV=<0.1 0.1>
    }
   }
   n3 = RigidBody(ComponentPose=@n2, bAlphaBoolEnabled=NOT Boolean(A=Get bIsSliding()), OverrideWorldGravity=Get GravityOverride()) {
    alphaInputType=Bool
    bClampLinearTranslationLimitToRefPose=true
    bOverrideWorldGravity=true
    overridePhysicsAsset=&/F_MED_Bountress_Coat_Physics
    simulationSpace=BaseBoneSpace
    alphaBoolBlend {blendInTime=0.2, blendOption=Linear, blendOutTime=0.2}
    baseBoneRef {boneName=pelvis}
    componentAppliedLinearAccClamp=<6000 6000 6000>
    componentLinearAccScale=<0.25 0.25 0.25>
    componentLinearVelScale=<0.25 0.25 0.25>
    overrideWorldGravity {z=-980}
    simSpaceSettings {
     maxAngularAcceleration=500
     maxAngularVelocity=3
     maxLinearAcceleration=400
     maxLinearVelocity=250
     velocityScaleZ=0
     worldAlpha=1
     externalLinearDragV=<0.1 0.1>
    }
   }
   n4 = Get backpack_wearing() -> (backpack_wearing:bool)
   n5 = Get GravityOverride_scarf() -> (GravityOverride_scarf:Vector)
   n6 = RigidBody(ComponentPose=@n3, bAlphaBoolEnabled=NOT Boolean(A=@n4), OverrideWorldGravity=@n5) {
    alphaInputType=Bool
    bClampLinearTranslationLimitToRefPose=true
    bOverrideWorldGravity=true
    overridePhysicsAsset=&/F_MED_Bountress_Coat_Physics_scarf
    simulationSpace=BaseBoneSpace
    alphaBoolBlend {blendOption=Linear}
    baseBoneRef {boneName=root}
    componentAppliedLinearAccClamp=<3000 3000 3000>
    componentLinearAccScale=<0.2 0.2 0.3>
    componentLinearVelScale=<0.2 0.2 0.3>
    simSpaceSettings {
     maxAngularAcceleration=50
     maxAngularVelocity=1
     maxLinearAcceleration=300
     maxLinearVelocity=150
     velocityScaleZ=0
     worldAlpha=1
     externalLinearDragV=<0.2 0.2>
    }
   }
   n7 = RigidBody(ComponentPose=@n6, bAlphaBoolEnabled=@n4, OverrideWorldGravity=@n5) {
    alphaInputType=Bool
    bClampLinearTranslationLimitToRefPose=true
    bOverrideWorldGravity=true
    overridePhysicsAsset=&/F_MED_Bountress_Coat_Physics_scarf_backpack
    simulationSpace=BaseBoneSpace
    alphaBoolBlend {blendOption=Linear}
    baseBoneRef {boneName=root}
    componentAppliedLinearAccClamp=<3000 3000 3000>
    componentLinearAccScale=<0.2 0.2 0.3>
    componentLinearVelScale=<0.2 0.2 0.3>
    simSpaceSettings {
     maxAngularAcceleration=50
     maxAngularVelocity=1
     maxLinearAcceleration=300
     maxLinearVelocity=150
     velocityScaleZ=0
     worldAlpha=1
     externalLinearDragV=<0.2 0.2>
    }
   }
   n8 = Anim Dynamics - Bone: dyn_spur_parent_l(ComponentPose=@n7, Alpha=1, ExternalForce="0,0,0") {
    angularDampingOverride=0.9999
    angularSpringConstant=0.5
    bGravityOverrideInSimSpace=true
    bOverrideAngularDamping=true
    bOverrideLinearDamping=true
    linearDampingOverride=0.9999
    simulationSpace=RootRelative
    alphaBoolBlend {blendOption=Linear}
    boundBone {boneName=dyn_spur_parent_l}
    componentAppliedLinearAccClamp=<10000 10000 10000>
    componentLinearAccScale=<0.15 0.15 0.15>
    componentLinearVelScale=<0.15 0.15 0.15>
    gravityOverride=<1 1 1>
    physicsBodyDefinitions [{
      sphereCollisionRadius=0
      boundBone {boneName=dyn_spur_parent_l}
      boxExtents=<6 6 2>
      constraintSetup {angularTargetAxis=AxisZ, angularLimitsMax=<0 0 10>, angularLimitsMin=<0 0 -10>, angularTarget=<0 0 0>}
      localJointOffset=<-1 0 1>
     }]
    relativeSpaceBone {boneName=root}
   }
   n9 = Anim Dynamics - Bone: dyn_spur_parent_r(ComponentPose=@n8, Alpha=1, ExternalForce="0,0,0") {
    angularDampingOverride=0.9999
    angularSpringConstant=0.5
    bGravityOverrideInSimSpace=true
    bOverrideAngularDamping=true
    bOverrideLinearDamping=true
    linearDampingOverride=0.9999
    simulationSpace=RootRelative
    alphaBoolBlend {blendOption=Linear}
    boundBone {boneName=dyn_spur_parent_r}
    componentAppliedLinearAccClamp=<10000 10000 10000>
    componentLinearAccScale=<0.15 0.15 0.15>
    componentLinearVelScale=<0.15 0.15 0.15>
    gravityOverride=<1 1 1>
    physicsBodyDefinitions [{
      sphereCollisionRadius=0
      boundBone {boneName=dyn_spur_parent_r}
      boxExtents=<6 6 2>
      constraintSetup {angularLimitsMax=<0 0 10>, angularLimitsMin=<0 0 -10>, angularTarget=<0 0 0>}
      localJointOffset=<-1 0 1>
     }]
    relativeSpaceBone {boneName=root}
   }
   n10 = Control Rig!(
    Source=Component To Local(ComponentPose=@n9)
    Alpha
    sliding=Get sliding()
    RBAN=Get bRigidBodyEnabled()
    crouching=Get f_crouching()
    f_crouching_and_moving_2D=Get f_crouching_and_moving_2D()
    knee_sliding:double=Get knee_sliding() -> (knee_sliding:double)
    baseball_sliding:double=Get baseball_sliding() -> (baseball_sliding:double)
    moving_emote:double=Get moving_emote() -> (moving_emote:double)
    cape_mod:bool=Get cape_mod() -> (cape_mod:bool)
   ) -> (Pose) {
    alphaBoolBlend {blendOption=Linear}
    controlRigAssetReference {blueprintRigClass=&/F_MED_Bountress_ControlRig.F_MED_Bountress_ControlRig_C}
   }
   Output Pose(Result=@n10) {layerGroup=None}
  ]}
 EventGraph {nodes [
   Event BlueprintUpdateAnimation()
   n1 = OR Boolean!(A=Get bIsPlayingEmote(), B=Get bIsMoving2D(), C:bool=NOT Boolean(A=Get bRigidBodyEnabled())) -> (ReturnValue)
   > Set moving_emote(moving_emote:double="To Float (Boolean)"(InBool=NOT Boolean(A=@n1))) -> (Output_Get:double)
   > Set knee_sliding(knee_sliding:double="To Float (Boolean)"(InBool=Get bIsKneeSliding())) -> (Output_Get:double)
   > Set baseball_sliding(baseball_sliding:double="To Float (Boolean)"(InBool=Get bIsBaseballSliding())) -> (Output_Get:double)
   n2 = Get bIsCrouching()
   > Set f_not_crouching(f_not_crouching:double="To Float (Boolean)"(InBool=NOT Boolean(A=@n2))) -> (Output_Get:double)
   n3 = Get bIsCrouching()
   n4 = Get bIsMoving2D()
   > Set f_crouching_and_moving_2D(f_crouching_and_moving_2D:double="To Float (Boolean)"(InBool=AND Boolean(A=@n4, B=@n3))) -> (Output_Get:double)
   > Set f_crouching(f_crouching:double="To Float (Boolean)"(InBool=Get bIsCrouching())) -> (Output_Get:double)
   > Set sliding(sliding:double="To Float (Boolean)"(InBool=Get bIsSliding())) -> (Output_Get:double)
   n5 = CheckForBackbling()
   > Set cape_mod(cape_mod:bool=@n5.bIsBackblingACape) -> (Output_Get:bool)
   > n6 = Set backpack_wearing(backpack_wearing:bool=@n5) -> (Output_Get:bool)
   > n7 = Sequence!(execute) -> (then_0, then_1, then_2:Exec, then_3:Exec, then_4:Exec)
   > n8 = Branch(Condition=@n6)
   > K2Node_Composite_4 = CollapseGraph_0_2() {is=title K2Node_Composite}
   K2Node_Composite_5 = CollapseGraph_2(execute=@n8.else) {is=title K2Node_Composite}
   n9 = Gravity Override State Function DEPRECATED(
    execute=@n7.then_1
    CharacterPartAnimInstance=Self-Reference()
    GameTimeInSeconds=GetGameTimeInSeconds()
    GravityOverrideParams=Get Gravity Override Params()
    DebugJointName=spine_05
   )
   > Set GravityOverride(GravityOverride=@n9)
   n10 = NOT Boolean(A=Get bIsSkyDiving())
   n11 = Get bIsFalling()
   n12 = Branch(execute=@n7.then_2, Condition=OR Boolean!(A=AND Boolean(A=@n10, B=@n11), B=Get bIsJumping()) -> (ReturnValue))
   > K2Node_Composite_2 = CollapseGraph_2_2_2() {is=title K2Node_Composite}
   K2Node_Composite_0 = CollapseGraph_2_2_2_2(execute=@n12.else) {is=title K2Node_Composite}
   n13 = Gravity Override State Function DEPRECATED(
    execute=@n7.then_3
    CharacterPartAnimInstance=Self-Reference()
    GameTimeInSeconds=GetGameTimeInSeconds()
    GravityOverrideParams=Get Gravity Override Params()
    DebugJointName=spine_05
   )
   > Set GravityOverride_scarf(GravityOverride_scarf:Vector=@n13) -> (Output_Get:Vector)
  ]}
 CollapseGraph_2_2_2_2 {
  path=EventGraph.K2Node_Composite_0.CollapseGraph_2_2_2_2
  nodes [
   Inputs()
   > Set!(
    execute
    Gravity Override Params_Skydive_GoingUp_28_#1_WindFrequency_28_#2:Vector="20,50,50"
    Gravity Override Params_Skydive_GoingUp_28_#1_WindAmplitude_31_#3:Vector="200,-500,50"
    Gravity Override Params_Skydive_GoingUp_28_#1_WindOffset_32_#4:Vector="0,0,50"
    Gravity Override Params_Skydive_GoingUp_28_#1_GravityOverride_15_#5:Vector="0,-500,500"
    Gravity Override Params_Skydive_GoingUp_28_#1_JointName_27_#6:Name=root
    Gravity Override Params_Skydive_FloatingDown_29_#7_WindFrequency_28_#2:Vector="1.5,7,50"
    Gravity Override Params_Skydive_FloatingDown_29_#7_WindAmplitude_31_#3:Vector="200,-500,50"
    Gravity Override Params_Skydive_FloatingDown_29_#7_WindOffset_32_#4:Vector="0,-600,50"
    Gravity Override Params_Skydive_FloatingDown_29_#7_GravityOverride_15_#5:Vector="0,-500,500"
    Gravity Override Params_Skydive_FloatingDown_29_#7_JointName_27_#6:Name=root
    Gravity Override Params_Skydive_DivingDown_30_#8_WindFrequency_28_#2:Vector="1.5,12,10"
    Gravity Override Params_Skydive_DivingDown_30_#8_WindAmplitude_31_#3:Vector="50,300,100"
    Gravity Override Params_Skydive_DivingDown_30_#8_WindOffset_32_#4:Vector="0,0,0"
    Gravity Override Params_Skydive_DivingDown_30_#8_GravityOverride_15_#5:Vector="-1500,0,0"
    Gravity Override Params_Skydive_DivingDown_30_#8_JointName_27_#6:Name=pelvis
    Gravity Override Params_Skydive_Parachuting_31_#9_WindFrequency_28_#2:Vector="1.5,7,8"
    Gravity Override Params_Skydive_Parachuting_31_#9_WindAmplitude_31_#3:Vector="200,-500,400"
    Gravity Override Params_Skydive_Parachuting_31_#9_WindOffset_32_#4:Vector="0,0,0"
    Gravity Override Params_Skydive_Parachuting_31_#9_GravityOverride_15_#5:Vector="0,-800,-900"
    Gravity Override Params_Skydive_Parachuting_31_#9_JointName_27_#6:Name=root
    Gravity Override Params_OnGround_Standing_32_#10:Gravity Override Setting Struct Structure=(WindFrequency_28_#2=(X=0,Y=0,Z=0),WindAmplitude_31_#3=(X=0,Y=0,Z=0),WindOffset_32_#4=(X=0,Y=0,Z=0),GravityOverride_15_#5=(X=0,Y=0,Z=-980),JointName_27_#6="root")
    Gravity Override Params_OnGround_Moving_33_#11:Gravity Override Setting Struct Structure=(WindFrequency_28_#2=(X=0,Y=0,Z=0),WindAmplitude_31_#3=(X=0,Y=0,Z=0),WindOffset_32_#4=(X=0,Y=0,Z=0),GravityOverride_15_#5=(X=0,Y=0,Z=-980),JointName_27_#6="root")
    Gravity Override Params_EmoteOrMelee_34_#12:Gravity Override Setting Struct Structure=(WindFrequency_28_#2=(X=0,Y=0,Z=0),WindAmplitude_31_#3=(X=0,Y=0,Z=0),WindOffset_32_#4=(X=0,Y=0,Z=0),GravityOverride_15_#5=(X=0,Y=0,Z=-980),JointName_27_#6="root")
    Gravity Override Params_Swimming_35_#13:Gravity Override Setting Struct Structure=(WindFrequency_28_#2=(X=0,Y=0,Z=0),WindAmplitude_31_#3=(X=0,Y=0,Z=0),WindOffset_32_#4=(X=0,Y=0,Z=0),GravityOverride_15_#5=(X=-800,Y=0,Z=0),JointName_27_#6="pelvis")
    Gravity Override Params_Front_End_36_#14:Gravity Override Setting Struct Structure=(WindFrequency_28_#2=(X=0,Y=0,Z=0),WindAmplitude_31_#3=(X=0,Y=0,Z=0),WindOffset_32_#4=(X=0,Y=0,Z=0),GravityOverride_15_#5=(X=0,Y=0,Z=-980),JointName_27_#6="root")
    Gravity Override Params_DBNO_40_#15:Gravity Override Setting Struct Structure=(WindFrequency_28_#2=(X=0,Y=0,Z=0),WindAmplitude_31_#3=(X=0,Y=0,Z=0),WindOffset_32_#4=(X=0,Y=0,Z=0),GravityOverride_15_#5=(X=-980,Y=400,Z=0),JointName_27_#6="pelvis")
    Gravity Override Params_DBNO_Carried_41_#16:Gravity Override Setting Struct Structure=(WindFrequency_28_#2=(X=0,Y=0,Z=0),WindAmplitude_31_#3=(X=0,Y=0,Z=0),WindOffset_32_#4=(X=0,Y=0,Z=0),GravityOverride_15_#5=(X=-980,Y=500,Z=-500),JointName_27_#6="pelvis")
    Gravity Override Params_OnGround_Falling_45_#17_WindFrequency_28_#2:Vector="0,0,0"
    Gravity Override Params_OnGround_Falling_45_#17_WindAmplitude_31_#3:Vector="0,0,0"
    Gravity Override Params_OnGround_Falling_45_#17_WindOffset_32_#4:Vector="0,0,0"
    Gravity Override Params_OnGround_Falling_45_#17_GravityOverride_15_#5:Vector="0,0,-900"
    Gravity Override Params_OnGround_Falling_45_#17_JointName_27_#6:Name=root
    Gravity Override Params_Ziplining_46_#18:Gravity Override Setting Struct Structure=(WindFrequency_28_#2=(X=0,Y=0,Z=0),WindAmplitude_31_#3=(X=0,Y=0,Z=0),WindOffset_32_#4=(X=0,Y=0,Z=0),GravityOverride_15_#5=(X=0,Y=0,Z=-980),JointName_27_#6="root")
    Gravity Override Params_Skydive_Surfing_49_#19_WindFrequency_28_#2:Vector="1.5,7,8"
    Gravity Override Params_Skydive_Surfing_49_#19_WindAmplitude_31_#3:Vector="200,-500,400"
    Gravity Override Params_Skydive_Surfing_49_#19_WindOffset_32_#4:Vector="0,0,0"
    Gravity Override Params_Skydive_Surfing_49_#19_GravityOverride_15_#5:Vector="500,-300,-900"
    Gravity Override Params_Skydive_Surfing_49_#19_JointName_27_#6:Name=root
    Gravity Override Params_Riding_Standing_54_#20:Gravity Override Setting Struct Structure=(WindFrequency_28_#2=(X=0,Y=0,Z=0),WindAmplitude_31_#3=(X=0,Y=0,Z=0),WindOffset_32_#4=(X=0,Y=0,Z=0),GravityOverride_15_#5=(X=0,Y=-350,Z=-980),JointName_27_#6="root")
    Gravity Override Params_Riding_Moving_55_#21:Gravity Override Setting Struct Structure=(WindFrequency_28_#2=(X=0,Y=0,Z=0),WindAmplitude_31_#3=(X=0,Y=0,Z=0),WindOffset_32_#4=(X=0,Y=0,Z=0),GravityOverride_15_#5=(X=0,Y=-450,Z=-450),JointName_27_#6="root")
    Gravity Override Params_Riding_Moving_Targeting_58_#22:Gravity Override Setting Struct Structure=(WindFrequency_28_#2=(X=0,Y=0,Z=0),WindAmplitude_31_#3=(X=0,Y=0,Z=0),WindOffset_32_#4=(X=0,Y=0,Z=0),GravityOverride_15_#5=(X=0,Y=-200,Z=-980),JointName_27_#6="root")
    Gravity Override Params_Tactical_Sprint_57_#23:Gravity Override Setting Struct Structure=(WindFrequency_28_#2=(X=0,Y=0,Z=0),WindAmplitude_31_#3=(X=0,Y=0,Z=0),WindOffset_32_#4=(X=0,Y=0,Z=0),GravityOverride_15_#5=(X=0,Y=-450,Z=-500),JointName_27_#6="root")
    Gravity Override Params_Rail_Grinding_61_#24:Gravity Override Setting Struct Structure=(WindFrequency_28_#2=(X=0,Y=0,Z=0),WindAmplitude_31_#3=(X=0,Y=0,Z=0),WindOffset_32_#4=(X=0,Y=0,Z=0),GravityOverride_15_#5=(X=0,Y=0,Z=-980),JointName_27_#6="root")
    Gravity Override Params_DBNO_Tumbling_67_#25:Gravity Override Setting Struct Structure=(WindFrequency_28_#2=(X=0,Y=0,Z=0),WindAmplitude_31_#3=(X=0,Y=0,Z=0),WindOffset_32_#4=(X=0,Y=0,Z=0),GravityOverride_15_#5=(X=-2000,Y=0,Z=0),JointName_27_#6="pelvis")
    Gravity Override Params_DBNO_RollFwd_68_#26:Gravity Override Setting Struct Structure=(WindFrequency_28_#2=(X=0,Y=0,Z=0),WindAmplitude_31_#3=(X=0,Y=0,Z=0),WindOffset_32_#4=(X=0,Y=0,Z=0),GravityOverride_15_#5=(X=-2000,Y=0,Z=0),JointName_27_#6="pelvis")
    Gravity Override Params_DBNO_RollBwd_69_#27:Gravity Override Setting Struct Structure=(WindFrequency_28_#2=(X=0,Y=0,Z=0),WindAmplitude_31_#3=(X=0,Y=0,Z=0),WindOffset_32_#4=(X=0,Y=0,Z=0),GravityOverride_15_#5=(X=-4000,Y=0,Z=0),JointName_27_#6="pelvis")
    Gravity Override Params_DBNO_Impulse_70_#28:Gravity Override Setting Struct Structure=(WindFrequency_28_#2=(X=0,Y=0,Z=0),WindAmplitude_31_#3=(X=0,Y=0,Z=0),WindOffset_32_#4=(X=0,Y=0,Z=0),GravityOverride_15_#5=(X=-2000,Y=0,Z=0),JointName_27_#6="pelvis")
   ) -> (then, Output_Get:Gravity Override Params Struct Structure)
   Outputs()
  ]
 }
 CollapseGraph_2_2_2 {
  path=EventGraph.K2Node_Composite_2.CollapseGraph_2_2_2
  nodes [
   Inputs()
   > Set!(
    execute
    Gravity Override Params_Skydive_GoingUp_28_#1_WindFrequency_28_#2="20,50,50"
    Gravity Override Params_Skydive_GoingUp_28_#1_WindAmplitude_31_#3="200,-500,50"
    Gravity Override Params_Skydive_GoingUp_28_#1_WindOffset_32_#4="0,0,50"
    Gravity Override Params_Skydive_GoingUp_28_#1_GravityOverride_15_#5="0,-500,500"
    Gravity Override Params_Skydive_GoingUp_28_#1_JointName_27_#6=root
    Gravity Override Params_Skydive_FloatingDown_29_#7_WindFrequency_28_#2="1.5,7,50"
    Gravity Override Params_Skydive_FloatingDown_29_#7_WindAmplitude_31_#3="200,-500,50"
    Gravity Override Params_Skydive_FloatingDown_29_#7_WindOffset_32_#4="0,-600,50"
    Gravity Override Params_Skydive_FloatingDown_29_#7_GravityOverride_15_#5="0,-500,500"
    Gravity Override Params_Skydive_FloatingDown_29_#7_JointName_27_#6=root
    Gravity Override Params_Skydive_DivingDown_30_#8_WindFrequency_28_#2="1.5,12,10"
    Gravity Override Params_Skydive_DivingDown_30_#8_WindAmplitude_31_#3="50,300,100"
    Gravity Override Params_Skydive_DivingDown_30_#8_WindOffset_32_#4="0,0,0"
    Gravity Override Params_Skydive_DivingDown_30_#8_GravityOverride_15_#5="-1500,0,0"
    Gravity Override Params_Skydive_DivingDown_30_#8_JointName_27_#6=pelvis
    Gravity Override Params_Skydive_Parachuting_31_#9_WindFrequency_28_#2="1.5,7,8"
    Gravity Override Params_Skydive_Parachuting_31_#9_WindAmplitude_31_#3="200,-500,400"
    Gravity Override Params_Skydive_Parachuting_31_#9_WindOffset_32_#4="0,0,0"
    Gravity Override Params_Skydive_Parachuting_31_#9_GravityOverride_15_#5="0,-800,-900"
    Gravity Override Params_Skydive_Parachuting_31_#9_JointName_27_#6=root
    Gravity Override Params_OnGround_Standing_32_#10_WindFrequency_28_#2:Vector="0,0,0"
    Gravity Override Params_OnGround_Standing_32_#10_WindAmplitude_31_#3:Vector="0,0,0"
    Gravity Override Params_OnGround_Standing_32_#10_WindOffset_32_#4:Vector="0,0,0"
    Gravity Override Params_OnGround_Standing_32_#10_GravityOverride_15_#5:Vector="0,0,-650"
    Gravity Override Params_OnGround_Standing_32_#10_JointName_27_#6:Name=root
    Gravity Override Params_OnGround_Moving_33_#11_WindFrequency_28_#2:Vector="0,0,0"
    Gravity Override Params_OnGround_Moving_33_#11_WindAmplitude_31_#3:Vector="0,0,0"
    Gravity Override Params_OnGround_Moving_33_#11_WindOffset_32_#4:Vector="0,0,0"
    Gravity Override Params_OnGround_Moving_33_#11_GravityOverride_15_#5:Vector="0,0,-650"
    Gravity Override Params_OnGround_Moving_33_#11_JointName_27_#6:Name=root
    Gravity Override Params_EmoteOrMelee_34_#12=(WindFrequency_28_#2=(X=0,Y=0,Z=0),WindAmplitude_31_#3=(X=0,Y=0,Z=0),WindOffset_32_#4=(X=0,Y=0,Z=0),GravityOverride_15_#5=(X=0,Y=0,Z=-980),JointName_27_#6="root")
    Gravity Override Params_Swimming_35_#13=(WindFrequency_28_#2=(X=0,Y=0,Z=0),WindAmplitude_31_#3=(X=0,Y=0,Z=0),WindOffset_32_#4=(X=0,Y=0,Z=0),GravityOverride_15_#5=(X=-800,Y=0,Z=0),JointName_27_#6="pelvis")
    Gravity Override Params_Front_End_36_#14=(WindFrequency_28_#2=(X=0,Y=0,Z=0),WindAmplitude_31_#3=(X=0,Y=0,Z=0),WindOffset_32_#4=(X=0,Y=0,Z=0),GravityOverride_15_#5=(X=0,Y=0,Z=-980),JointName_27_#6="root")
    Gravity Override Params_DBNO_40_#15=(WindFrequency_28_#2=(X=0,Y=0,Z=0),WindAmplitude_31_#3=(X=0,Y=0,Z=0),WindOffset_32_#4=(X=0,Y=0,Z=0),GravityOverride_15_#5=(X=-980,Y=400,Z=0),JointName_27_#6="pelvis")
    Gravity Override Params_DBNO_Carried_41_#16=(WindFrequency_28_#2=(X=0,Y=0,Z=0),WindAmplitude_31_#3=(X=0,Y=0,Z=0),WindOffset_32_#4=(X=0,Y=0,Z=0),GravityOverride_15_#5=(X=-980,Y=500,Z=-500),JointName_27_#6="pelvis")
    Gravity Override Params_OnGround_Falling_45_#17_WindFrequency_28_#2="0,0,0"
    Gravity Override Params_OnGround_Falling_45_#17_WindAmplitude_31_#3="0,0,0"
    Gravity Override Params_OnGround_Falling_45_#17_WindOffset_32_#4="0,0,0"
    Gravity Override Params_OnGround_Falling_45_#17_GravityOverride_15_#5="0,0,-900"
    Gravity Override Params_OnGround_Falling_45_#17_JointName_27_#6=root
    Gravity Override Params_Ziplining_46_#18=(WindFrequency_28_#2=(X=0,Y=0,Z=0),WindAmplitude_31_#3=(X=0,Y=0,Z=0),WindOffset_32_#4=(X=0,Y=0,Z=0),GravityOverride_15_#5=(X=0,Y=0,Z=-980),JointName_27_#6="root")
    Gravity Override Params_Skydive_Surfing_49_#19_WindFrequency_28_#2="1.5,7,8"
    Gravity Override Params_Skydive_Surfing_49_#19_WindAmplitude_31_#3="200,-500,400"
    Gravity Override Params_Skydive_Surfing_49_#19_WindOffset_32_#4="0,0,0"
    Gravity Override Params_Skydive_Surfing_49_#19_GravityOverride_15_#5="500,-300,-900"
    Gravity Override Params_Skydive_Surfing_49_#19_JointName_27_#6=root
    Gravity Override Params_Riding_Standing_54_#20=(WindFrequency_28_#2=(X=0,Y=0,Z=0),WindAmplitude_31_#3=(X=0,Y=0,Z=0),WindOffset_32_#4=(X=0,Y=0,Z=0),GravityOverride_15_#5=(X=0,Y=-350,Z=-980),JointName_27_#6="root")
    Gravity Override Params_Riding_Moving_55_#21=(WindFrequency_28_#2=(X=0,Y=0,Z=0),WindAmplitude_31_#3=(X=0,Y=0,Z=0),WindOffset_32_#4=(X=0,Y=0,Z=0),GravityOverride_15_#5=(X=0,Y=-450,Z=-450),JointName_27_#6="root")
    Gravity Override Params_Riding_Moving_Targeting_58_#22=(WindFrequency_28_#2=(X=0,Y=0,Z=0),WindAmplitude_31_#3=(X=0,Y=0,Z=0),WindOffset_32_#4=(X=0,Y=0,Z=0),GravityOverride_15_#5=(X=0,Y=-200,Z=-980),JointName_27_#6="root")
    Gravity Override Params_Tactical_Sprint_57_#23=(WindFrequency_28_#2=(X=0,Y=0,Z=0),WindAmplitude_31_#3=(X=0,Y=0,Z=0),WindOffset_32_#4=(X=0,Y=0,Z=0),GravityOverride_15_#5=(X=0,Y=-450,Z=-500),JointName_27_#6="root")
    Gravity Override Params_Rail_Grinding_61_#24=(WindFrequency_28_#2=(X=0,Y=0,Z=0),WindAmplitude_31_#3=(X=0,Y=0,Z=0),WindOffset_32_#4=(X=0,Y=0,Z=0),GravityOverride_15_#5=(X=0,Y=0,Z=-980),JointName_27_#6="root")
    Gravity Override Params_DBNO_Tumbling_67_#25=(WindFrequency_28_#2=(X=0,Y=0,Z=0),WindAmplitude_31_#3=(X=0,Y=0,Z=0),WindOffset_32_#4=(X=0,Y=0,Z=0),GravityOverride_15_#5=(X=-2000,Y=0,Z=0),JointName_27_#6="pelvis")
    Gravity Override Params_DBNO_RollFwd_68_#26=(WindFrequency_28_#2=(X=0,Y=0,Z=0),WindAmplitude_31_#3=(X=0,Y=0,Z=0),WindOffset_32_#4=(X=0,Y=0,Z=0),GravityOverride_15_#5=(X=-2000,Y=0,Z=0),JointName_27_#6="pelvis")
    Gravity Override Params_DBNO_RollBwd_69_#27=(WindFrequency_28_#2=(X=0,Y=0,Z=0),WindAmplitude_31_#3=(X=0,Y=0,Z=0),WindOffset_32_#4=(X=0,Y=0,Z=0),GravityOverride_15_#5=(X=-4000,Y=0,Z=0),JointName_27_#6="pelvis")
    Gravity Override Params_DBNO_Impulse_70_#28=(WindFrequency_28_#2=(X=0,Y=0,Z=0),WindAmplitude_31_#3=(X=0,Y=0,Z=0),WindOffset_32_#4=(X=0,Y=0,Z=0),GravityOverride_15_#5=(X=-2000,Y=0,Z=0),JointName_27_#6="pelvis")
   ) -> (then, Output_Get)
   Outputs()
  ]
 }
 CollapseGraph_0_2 {
  path=EventGraph.K2Node_Composite_4.CollapseGraph_0_2
  nodes [
   Inputs()
   > Set!(
    execute
    Gravity Override Params_Skydive_GoingUp_28_#1_WindFrequency_28_#2="20,50,50"
    Gravity Override Params_Skydive_GoingUp_28_#1_WindAmplitude_31_#3="200,-500,50"
    Gravity Override Params_Skydive_GoingUp_28_#1_WindOffset_32_#4="0,0,50"
    Gravity Override Params_Skydive_GoingUp_28_#1_GravityOverride_15_#5="0,-500,50"
    Gravity Override Params_Skydive_GoingUp_28_#1_JointName_27_#6=root
    Gravity Override Params_Skydive_FloatingDown_29_#7_WindFrequency_28_#2="1.5,7,50"
    Gravity Override Params_Skydive_FloatingDown_29_#7_WindAmplitude_31_#3="200,-500,50"
    Gravity Override Params_Skydive_FloatingDown_29_#7_WindOffset_32_#4="0,-600,50"
    Gravity Override Params_Skydive_FloatingDown_29_#7_GravityOverride_15_#5="0,-500,50"
    Gravity Override Params_Skydive_FloatingDown_29_#7_JointName_27_#6=root
    Gravity Override Params_Skydive_DivingDown_30_#8_WindFrequency_28_#2="1.5,12,10"
    Gravity Override Params_Skydive_DivingDown_30_#8_WindAmplitude_31_#3="50,300,100"
    Gravity Override Params_Skydive_DivingDown_30_#8_WindOffset_32_#4="0,0,0"
    Gravity Override Params_Skydive_DivingDown_30_#8_GravityOverride_15_#5="0,1000,0"
    Gravity Override Params_Skydive_DivingDown_30_#8_JointName_27_#6=pelvis
    Gravity Override Params_Skydive_Parachuting_31_#9_WindFrequency_28_#2="1.5,7,8"
    Gravity Override Params_Skydive_Parachuting_31_#9_WindAmplitude_31_#3="200,-500,400"
    Gravity Override Params_Skydive_Parachuting_31_#9_WindOffset_32_#4="0,0,0"
    Gravity Override Params_Skydive_Parachuting_31_#9_GravityOverride_15_#5="0,-800,-900"
    Gravity Override Params_Skydive_Parachuting_31_#9_JointName_27_#6=root
    Gravity Override Params_OnGround_Standing_32_#10:Gravity Override Setting Struct Structure=(WindFrequency_28_#2=(X=0,Y=0,Z=0),WindAmplitude_31_#3=(X=0,Y=0,Z=0),WindOffset_32_#4=(X=0,Y=0,Z=0),GravityOverride_15_#5=(X=0,Y=0,Z=-980),JointName_27_#6="root")
    Gravity Override Params_OnGround_Moving_33_#11_WindFrequency_28_#2="0,0,0"
    Gravity Override Params_OnGround_Moving_33_#11_WindAmplitude_31_#3="0,0,0"
    Gravity Override Params_OnGround_Moving_33_#11_WindOffset_32_#4="0,0,0"
    Gravity Override Params_OnGround_Moving_33_#11_GravityOverride_15_#5="0,0,-900"
    Gravity Override Params_OnGround_Moving_33_#11_JointName_27_#6=root
    Gravity Override Params_EmoteOrMelee_34_#12=(WindFrequency_28_#2=(X=0,Y=0,Z=0),WindAmplitude_31_#3=(X=0,Y=0,Z=0),WindOffset_32_#4=(X=0,Y=0,Z=0),GravityOverride_15_#5=(X=0,Y=0,Z=-980),JointName_27_#6="root")
    Gravity Override Params_Swimming_35_#13=(WindFrequency_28_#2=(X=0,Y=0,Z=0),WindAmplitude_31_#3=(X=0,Y=0,Z=0),WindOffset_32_#4=(X=0,Y=0,Z=0),GravityOverride_15_#5=(X=-800,Y=0,Z=0),JointName_27_#6="pelvis")
    Gravity Override Params_Front_End_36_#14=(WindFrequency_28_#2=(X=0,Y=0,Z=0),WindAmplitude_31_#3=(X=0,Y=0,Z=0),WindOffset_32_#4=(X=0,Y=0,Z=0),GravityOverride_15_#5=(X=0,Y=0,Z=-980),JointName_27_#6="root")
    Gravity Override Params_DBNO_40_#15=(WindFrequency_28_#2=(X=0,Y=0,Z=0),WindAmplitude_31_#3=(X=0,Y=0,Z=0),WindOffset_32_#4=(X=0,Y=0,Z=0),GravityOverride_15_#5=(X=-980,Y=400,Z=0),JointName_27_#6="pelvis")
    Gravity Override Params_DBNO_Carried_41_#16=(WindFrequency_28_#2=(X=0,Y=0,Z=0),WindAmplitude_31_#3=(X=0,Y=0,Z=0),WindOffset_32_#4=(X=0,Y=0,Z=0),GravityOverride_15_#5=(X=-980,Y=500,Z=-500),JointName_27_#6="pelvis")
    Gravity Override Params_OnGround_Falling_45_#17_WindFrequency_28_#2="0,0,0"
    Gravity Override Params_OnGround_Falling_45_#17_WindAmplitude_31_#3="0,0,0"
    Gravity Override Params_OnGround_Falling_45_#17_WindOffset_32_#4="0,0,0"
    Gravity Override Params_OnGround_Falling_45_#17_GravityOverride_15_#5="0,0,-900"
    Gravity Override Params_OnGround_Falling_45_#17_JointName_27_#6=root
    Gravity Override Params_Ziplining_46_#18=(WindFrequency_28_#2=(X=0,Y=0,Z=0),WindAmplitude_31_#3=(X=0,Y=0,Z=0),WindOffset_32_#4=(X=0,Y=0,Z=0),GravityOverride_15_#5=(X=0,Y=0,Z=-980),JointName_27_#6="root")
    Gravity Override Params_Skydive_Surfing_49_#19_WindFrequency_28_#2="1.5,7,8"
    Gravity Override Params_Skydive_Surfing_49_#19_WindAmplitude_31_#3="200,-500,400"
    Gravity Override Params_Skydive_Surfing_49_#19_WindOffset_32_#4="0,0,0"
    Gravity Override Params_Skydive_Surfing_49_#19_GravityOverride_15_#5="500,-300,-980"
    Gravity Override Params_Skydive_Surfing_49_#19_JointName_27_#6=root
    Gravity Override Params_Riding_Standing_54_#20=(WindFrequency_28_#2=(X=0,Y=0,Z=0),WindAmplitude_31_#3=(X=0,Y=0,Z=0),WindOffset_32_#4=(X=0,Y=0,Z=0),GravityOverride_15_#5=(X=0,Y=-350,Z=-980),JointName_27_#6="root")
    Gravity Override Params_Riding_Moving_55_#21=(WindFrequency_28_#2=(X=0,Y=0,Z=0),WindAmplitude_31_#3=(X=0,Y=0,Z=0),WindOffset_32_#4=(X=0,Y=0,Z=0),GravityOverride_15_#5=(X=0,Y=-450,Z=-450),JointName_27_#6="root")
    Gravity Override Params_Riding_Moving_Targeting_58_#22=(WindFrequency_28_#2=(X=0,Y=0,Z=0),WindAmplitude_31_#3=(X=0,Y=0,Z=0),WindOffset_32_#4=(X=0,Y=0,Z=0),GravityOverride_15_#5=(X=0,Y=-200,Z=-980),JointName_27_#6="root")
    Gravity Override Params_Tactical_Sprint_57_#23_WindFrequency_28_#2:Vector="0,0,0"
    Gravity Override Params_Tactical_Sprint_57_#23_WindAmplitude_31_#3:Vector="0,0,0"
    Gravity Override Params_Tactical_Sprint_57_#23_WindOffset_32_#4:Vector="0,0,0"
    Gravity Override Params_Tactical_Sprint_57_#23_GravityOverride_15_#5:Vector="0,-450,-1800"
    Gravity Override Params_Tactical_Sprint_57_#23_JointName_27_#6:Name=root
    Gravity Override Params_Rail_Grinding_61_#24=(WindFrequency_28_#2=(X=0,Y=0,Z=0),WindAmplitude_31_#3=(X=0,Y=0,Z=0),WindOffset_32_#4=(X=0,Y=0,Z=0),GravityOverride_15_#5=(X=0,Y=0,Z=-980),JointName_27_#6="root")
    Gravity Override Params_DBNO_Tumbling_67_#25=(WindFrequency_28_#2=(X=0,Y=0,Z=0),WindAmplitude_31_#3=(X=0,Y=0,Z=0),WindOffset_32_#4=(X=0,Y=0,Z=0),GravityOverride_15_#5=(X=-2000,Y=0,Z=0),JointName_27_#6="pelvis")
    Gravity Override Params_DBNO_RollFwd_68_#26=(WindFrequency_28_#2=(X=0,Y=0,Z=0),WindAmplitude_31_#3=(X=0,Y=0,Z=0),WindOffset_32_#4=(X=0,Y=0,Z=0),GravityOverride_15_#5=(X=-2000,Y=0,Z=0),JointName_27_#6="pelvis")
    Gravity Override Params_DBNO_RollBwd_69_#27=(WindFrequency_28_#2=(X=0,Y=0,Z=0),WindAmplitude_31_#3=(X=0,Y=0,Z=0),WindOffset_32_#4=(X=0,Y=0,Z=0),GravityOverride_15_#5=(X=-4000,Y=0,Z=0),JointName_27_#6="pelvis")
    Gravity Override Params_DBNO_Impulse_70_#28=(WindFrequency_28_#2=(X=0,Y=0,Z=0),WindAmplitude_31_#3=(X=0,Y=0,Z=0),WindOffset_32_#4=(X=0,Y=0,Z=0),GravityOverride_15_#5=(X=-2000,Y=0,Z=0),JointName_27_#6="pelvis")
   ) -> (then, Output_Get)
   Outputs()
  ]
 }
 CollapseGraph_2 {
  path=EventGraph.K2Node_Composite_5.CollapseGraph_2
  nodes [
   Inputs()
   > Set!(
    execute
    Gravity Override Params_Skydive_GoingUp_28_#1_WindFrequency_28_#2="20,50,50"
    Gravity Override Params_Skydive_GoingUp_28_#1_WindAmplitude_31_#3="200,-500,50"
    Gravity Override Params_Skydive_GoingUp_28_#1_WindOffset_32_#4="0,0,50"
    Gravity Override Params_Skydive_GoingUp_28_#1_GravityOverride_15_#5="0,-500,500"
    Gravity Override Params_Skydive_GoingUp_28_#1_JointName_27_#6=root
    Gravity Override Params_Skydive_FloatingDown_29_#7_WindFrequency_28_#2="1.5,7,50"
    Gravity Override Params_Skydive_FloatingDown_29_#7_WindAmplitude_31_#3="200,-500,50"
    Gravity Override Params_Skydive_FloatingDown_29_#7_WindOffset_32_#4="0,-600,50"
    Gravity Override Params_Skydive_FloatingDown_29_#7_GravityOverride_15_#5="0,-500,500"
    Gravity Override Params_Skydive_FloatingDown_29_#7_JointName_27_#6=root
    Gravity Override Params_Skydive_DivingDown_30_#8_WindFrequency_28_#2="1.5,12,10"
    Gravity Override Params_Skydive_DivingDown_30_#8_WindAmplitude_31_#3="50,300,100"
    Gravity Override Params_Skydive_DivingDown_30_#8_WindOffset_32_#4="0,0,0"
    Gravity Override Params_Skydive_DivingDown_30_#8_GravityOverride_15_#5="-1500,0,0"
    Gravity Override Params_Skydive_DivingDown_30_#8_JointName_27_#6=pelvis
    Gravity Override Params_Skydive_Parachuting_31_#9_WindFrequency_28_#2="1.5,7,8"
    Gravity Override Params_Skydive_Parachuting_31_#9_WindAmplitude_31_#3="200,-500,400"
    Gravity Override Params_Skydive_Parachuting_31_#9_WindOffset_32_#4="0,0,0"
    Gravity Override Params_Skydive_Parachuting_31_#9_GravityOverride_15_#5="0,-800,-900"
    Gravity Override Params_Skydive_Parachuting_31_#9_JointName_27_#6=root
    Gravity Override Params_OnGround_Standing_32_#10=(WindFrequency_28_#2=(X=0,Y=0,Z=0),WindAmplitude_31_#3=(X=0,Y=0,Z=0),WindOffset_32_#4=(X=0,Y=0,Z=0),GravityOverride_15_#5=(X=0,Y=0,Z=-980),JointName_27_#6="root")
    Gravity Override Params_OnGround_Moving_33_#11:Gravity Override Setting Struct Structure=(WindFrequency_28_#2=(X=0,Y=0,Z=0),WindAmplitude_31_#3=(X=0,Y=0,Z=0),WindOffset_32_#4=(X=0,Y=0,Z=0),GravityOverride_15_#5=(X=0,Y=0,Z=-980),JointName_27_#6="root")
    Gravity Override Params_EmoteOrMelee_34_#12=(WindFrequency_28_#2=(X=0,Y=0,Z=0),WindAmplitude_31_#3=(X=0,Y=0,Z=0),WindOffset_32_#4=(X=0,Y=0,Z=0),GravityOverride_15_#5=(X=0,Y=0,Z=-980),JointName_27_#6="root")
    Gravity Override Params_Swimming_35_#13=(WindFrequency_28_#2=(X=0,Y=0,Z=0),WindAmplitude_31_#3=(X=0,Y=0,Z=0),WindOffset_32_#4=(X=0,Y=0,Z=0),GravityOverride_15_#5=(X=-800,Y=0,Z=0),JointName_27_#6="pelvis")
    Gravity Override Params_Front_End_36_#14=(WindFrequency_28_#2=(X=0,Y=0,Z=0),WindAmplitude_31_#3=(X=0,Y=0,Z=0),WindOffset_32_#4=(X=0,Y=0,Z=0),GravityOverride_15_#5=(X=0,Y=0,Z=-980),JointName_27_#6="root")
    Gravity Override Params_DBNO_40_#15=(WindFrequency_28_#2=(X=0,Y=0,Z=0),WindAmplitude_31_#3=(X=0,Y=0,Z=0),WindOffset_32_#4=(X=0,Y=0,Z=0),GravityOverride_15_#5=(X=-980,Y=400,Z=0),JointName_27_#6="pelvis")
    Gravity Override Params_DBNO_Carried_41_#16=(WindFrequency_28_#2=(X=0,Y=0,Z=0),WindAmplitude_31_#3=(X=0,Y=0,Z=0),WindOffset_32_#4=(X=0,Y=0,Z=0),GravityOverride_15_#5=(X=-980,Y=500,Z=-500),JointName_27_#6="pelvis")
    Gravity Override Params_OnGround_Falling_45_#17_WindFrequency_28_#2="0,0,0"
    Gravity Override Params_OnGround_Falling_45_#17_WindAmplitude_31_#3="0,0,0"
    Gravity Override Params_OnGround_Falling_45_#17_WindOffset_32_#4="0,0,0"
    Gravity Override Params_OnGround_Falling_45_#17_GravityOverride_15_#5="0,0,-900"
    Gravity Override Params_OnGround_Falling_45_#17_JointName_27_#6=root
    Gravity Override Params_Ziplining_46_#18=(WindFrequency_28_#2=(X=0,Y=0,Z=0),WindAmplitude_31_#3=(X=0,Y=0,Z=0),WindOffset_32_#4=(X=0,Y=0,Z=0),GravityOverride_15_#5=(X=0,Y=0,Z=-980),JointName_27_#6="root")
    Gravity Override Params_Skydive_Surfing_49_#19_WindFrequency_28_#2="1.5,7,8"
    Gravity Override Params_Skydive_Surfing_49_#19_WindAmplitude_31_#3="200,-500,400"
    Gravity Override Params_Skydive_Surfing_49_#19_WindOffset_32_#4="0,0,0"
    Gravity Override Params_Skydive_Surfing_49_#19_GravityOverride_15_#5="500,-300,-900"
    Gravity Override Params_Skydive_Surfing_49_#19_JointName_27_#6=root
    Gravity Override Params_Riding_Standing_54_#20=(WindFrequency_28_#2=(X=0,Y=0,Z=0),WindAmplitude_31_#3=(X=0,Y=0,Z=0),WindOffset_32_#4=(X=0,Y=0,Z=0),GravityOverride_15_#5=(X=0,Y=-350,Z=-980),JointName_27_#6="root")
    Gravity Override Params_Riding_Moving_55_#21=(WindFrequency_28_#2=(X=0,Y=0,Z=0),WindAmplitude_31_#3=(X=0,Y=0,Z=0),WindOffset_32_#4=(X=0,Y=0,Z=0),GravityOverride_15_#5=(X=0,Y=-450,Z=-450),JointName_27_#6="root")
    Gravity Override Params_Riding_Moving_Targeting_58_#22=(WindFrequency_28_#2=(X=0,Y=0,Z=0),WindAmplitude_31_#3=(X=0,Y=0,Z=0),WindOffset_32_#4=(X=0,Y=0,Z=0),GravityOverride_15_#5=(X=0,Y=-200,Z=-980),JointName_27_#6="root")
    Gravity Override Params_Tactical_Sprint_57_#23:Gravity Override Setting Struct Structure=(WindFrequency_28_#2=(X=0,Y=0,Z=0),WindAmplitude_31_#3=(X=0,Y=0,Z=0),WindOffset_32_#4=(X=0,Y=0,Z=0),GravityOverride_15_#5=(X=0,Y=-450,Z=-500),JointName_27_#6="root")
    Gravity Override Params_Rail_Grinding_61_#24=(WindFrequency_28_#2=(X=0,Y=0,Z=0),WindAmplitude_31_#3=(X=0,Y=0,Z=0),WindOffset_32_#4=(X=0,Y=0,Z=0),GravityOverride_15_#5=(X=0,Y=0,Z=-980),JointName_27_#6="root")
    Gravity Override Params_DBNO_Tumbling_67_#25=(WindFrequency_28_#2=(X=0,Y=0,Z=0),WindAmplitude_31_#3=(X=0,Y=0,Z=0),WindOffset_32_#4=(X=0,Y=0,Z=0),GravityOverride_15_#5=(X=-2000,Y=0,Z=0),JointName_27_#6="pelvis")
    Gravity Override Params_DBNO_RollFwd_68_#26=(WindFrequency_28_#2=(X=0,Y=0,Z=0),WindAmplitude_31_#3=(X=0,Y=0,Z=0),WindOffset_32_#4=(X=0,Y=0,Z=0),GravityOverride_15_#5=(X=-2000,Y=0,Z=0),JointName_27_#6="pelvis")
    Gravity Override Params_DBNO_RollBwd_69_#27=(WindFrequency_28_#2=(X=0,Y=0,Z=0),WindAmplitude_31_#3=(X=0,Y=0,Z=0),WindOffset_32_#4=(X=0,Y=0,Z=0),GravityOverride_15_#5=(X=-4000,Y=0,Z=0),JointName_27_#6="pelvis")
    Gravity Override Params_DBNO_Impulse_70_#28=(WindFrequency_28_#2=(X=0,Y=0,Z=0),WindAmplitude_31_#3=(X=0,Y=0,Z=0),WindOffset_32_#4=(X=0,Y=0,Z=0),GravityOverride_15_#5=(X=-2000,Y=0,Z=0),JointName_27_#6="pelvis")
   ) -> (then, Output_Get)
   Outputs()
  ]
 }
]
variables [
 backpack_wearing {type=boolean}
 baseball_sliding {type=number}
 cape_mod {type=boolean}
 f_crouching {type=number}
 f_crouching_and_moving_2D {type=number}
 f_not_crouching {type=number}
 Gravity Override Params {type=GravityOverrideParamsStruct}
 gravity_mod_scarf {type=Vector}
 GravityOverride {type=Vector}
 GravityOverride_scarf {type=Vector}
 knee_sliding {type=number}
 moving_emote {type=number}
 RBAN {type=boolean}
 sliding {type=number}
]
