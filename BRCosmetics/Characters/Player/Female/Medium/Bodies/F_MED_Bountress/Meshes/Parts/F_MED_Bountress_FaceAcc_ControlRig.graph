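# Control Rig for the Bountress face accessory mesh that adjusts the back hair bone as the head turns or tilts.
# A spherical pose reader on the head outputs a weight, smoothed by alpha interpolation, that blends a rotation and translation offset onto dyn_back_hair_01.
# A following mid-hair modify applies an identity transform, and an extra_head offset node sits disconnected from execution.
# The pose reader has debug drawing enabled.
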ControlRigBlueprint /BRCosmetics/Characters/Player/Female/Medium/Bodies/F_MED_Bountress/Meshes/Parts/F_MED_Bountress_FaceAcc_ControlRig
graphs [
 RigVMFunctionLibrary {links [], nodes []}
 RigVMModel {nodes [
   RigUnit_BeginExecution = &RigUnit_BeginExecution()
   n1 = &RigUnit_SphericalPoseReader(
    ActiveRegionScaleFactors=(PositiveWidth=0.8,NegativeWidth=0.8,PositiveHeight=0.8,NegativeHeight=0.8)
    ActiveRegionSize=0.1
    Debug=(bDrawDebug=true,bDraw2D=false,bDrawLocalAxes=false,DebugScale=25,DebugSegments=20,DebugThickness=0.25)
    DriverAxis=(X=1,Y=0,Z=0)
    DriverItem=(Type=Bone,Name="head")
    FalloffRegionScaleFactors=(PositiveWidth=0.2,NegativeWidth=0.2,PositiveHeight=0.2,NegativeHeight=0.2)
    FalloffSize=1
    FlipWidthScaling=False
    OptionalParentItem=(Type=Bone)
    RotationOffset=(X=-50,Y=-90,Z=0)
    ExecutePin=@RigUnit_BeginExecution.ExecutePin
   )
   n2 = &RigVMFunction_AlphaInterp(InterpSpeedDecreasing=5, InterpSpeedIncreasing=5, Scale=1, bClampResult=False, Value=@n1.OutputParam)
   > &RigUnit_ModifyTransforms(
    ItemToModify=((Item=(Type=Bone,Name="dyn_back_hair_01"),Transform=(Rotation=(X=0,Y=-0.1305,Z=0,W=0.9914),Translation=(X=-1,Y=0,Z=0),Scale3D=(X=1,Y=1,Z=1))))
    Weight=@n2.Result
   )
   > &RigUnit_ModifyTransforms(
    ItemToModify=((Item=(Type=Bone,Name="dyn_mid_hair_01"),Transform=(Rotation=(X=0,Y=0,Z=0,W=1),Translation=(X=0,Y=0,Z=0),Scale3D=(X=1,Y=1,Z=1))))
    Weight=1
   )
   &RigUnit_ModifyTransforms(
    ExecutePin=()
    ItemToModify=((Item=(Type=Bone,Name="extra_head"),Transform=(Rotation=(X=0,Y=0,Z=0,W=1),Translation=(X=10,Y=0,Z=0),Scale3D=(X=1,Y=1,Z=1))))
    Weight=1
   )
  ]}
]
hierarchy [
 root {loc=<0 0 0>, rot=<0 0 0 1>}
 pelvis {parent=root, loc=<0 -0.5301 98.0571>, rot=<0.0127 0.707 -0.0127 -0.707>}
 spine_01 {parent=pelvis, loc=<6.1208 -0.9328 0>, rot=<0 0 0.034 -0.9994>}
 spine_02 {parent=spine_01, loc=<6.5996 -0.2095 0>, rot=<0 0 0.0238 -0.9997>}
 spine_03 {parent=spine_02, loc=<6.4914 -0.1237 0>, rot=<0 0 0.0149 -0.9999>}
 spine_04 {parent=spine_03, loc=<6.9598 0.0794 0>, rot=<0 0 -0.0132 -0.9999>}
 spine_05 {parent=spine_04, loc=<6.9987 0 0>, rot=<0 0 -0.0015 -1>}
 neck_01 {parent=spine_05, loc=<11.3138 0.1529 0>, rot=<0 0 -0.2003 -0.9797>}
 neck_02 {parent=neck_01, loc=<4.5627 0 0>, rot=<0 0 0.0504 -0.9987>}
 head {parent=neck_02, loc=<5.3914 0 0>, rot=<0 0 0.115 -0.9934>}
 dyn_front_hair_01 {parent=head, loc=<5.2808 5.9369 6.7684>, rot=<0.0041 -0.996 0.0888 -0.0004>}
 dyn_front_hair_02 {parent=dyn_front_hair_01, loc=<4.7389 0 0>, rot=<0 0 0 1>}
 dyn_mid_hair_01 {parent=head, loc=<7.021 -2.2518 7.5505>, rot=<0.004 -0.9704 -0.2414 0.001>}
 dyn_mid_hair_02 {parent=dyn_mid_hair_01, loc=<4.228 0 0>, rot=<0 0 0 1>}
 dyn_back_hair_01 {parent=head, loc=<7.0676 -7.8812 2.7743>, rot=<0.0031 -0.7567 -0.6538 0.0027>}
 dyn_back_hair_02 {parent=dyn_back_hair_01, loc=<4.228 0 0>, rot=<0 0 0 1>}
 extra_head {parent=head, loc=<0 0 0>, rot=<0 0 0 1>}
]
