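# Physics asset for the Bountress female medium body's face accessory mesh, setting up simulated hair and head collision shapes.
# Back, mid and front hair each form a two-bone capsule chain hung from the head, with narrow swing and twist limits.
# The head carries two capsules, and an extra_head capsule body joins it through a constraint with free swing and twist.
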PhysicsAsset /BRCosmetics/Characters/Player/Female/Medium/Bodies/F_MED_Bountress/Meshes/Parts/F_MED_Bountress_FaceAcc_Physics
bodies {
 dyn_back_hair_01 [{
   shapeType=Capsule
   length=5.5625
   radius=2.6797
   shapeName=dyn_hair_02_capsule
   center=<2.9648 -0.3612 1.0204>
   rotation=<rot -90 176.4237 -88.5779>
  }]
 dyn_back_hair_02 [{
   shapeType=Capsule
   length=4.6508
   radius=2.5679
   shapeName=dyn_hair_03_capsule
   center=<2.6506 -0.0209 0.8202>
   rotation=<rot 90 -165.9638 -74.426>
  }]
 dyn_front_hair_01 [{
   shapeType=Capsule
   length=1.1875
   radius=2.1562
   shapeName=dyn_hair_02_r_capsule
   center=<2.7274 0.1629 -0.0217>
   rotation=<rot 90 -21.1366 -20.3111>
  }]
 dyn_front_hair_02 [{
   shapeType=Capsule
   length=4.8956
   radius=1.9169
   shapeName=dyn_hair_03_r_capsule
   center=<3.4576 -1.2718 0.2929>
   rotation=<rot 84.3811 -4.0098 -0.0003>
  }]
 dyn_mid_hair_01 [{
   shapeType=Capsule
   length=4.4706
   radius=2.2044
   shapeName=dyn_hair_back_02_r_capsule
   center=<3.7414 -0.1745 0.9083>
   rotation=<rot -85.0002 0.0002 88.3037>
  }]
 dyn_mid_hair_02 [{
   shapeType=Capsule
   length=4.6875
   radius=2
   shapeName=dyn_hair_back_03_r_capsule
   center=<2.9305 -0.1886 0.5339>
   rotation=<rot 85.0001 -179.9997 92.1864>
  }]
 extra_head [{
   shapeType=Capsule
   length=6.3331
   radius=8.1332
   shapeName=extra_head_capsule
   center=<-12.2703 -2.2993 -0.016>
   rotation=<rot -89.6001 90.0238 144.0775>
  }]
 head [
  {shapeType=Capsule, length=7.7534, radius=5.3369, center=<3.1734 2.6019 0.0541>, rotation=<rot 0.5016 -90.415 -88.3249>}
  {
   shapeType=Capsule
   length=2.7534
   radius=3.8369
   shapeName=head_capsule
   center=<1.1072 -0.3919 0.4333>
   rotation=<rot 0.5016 -90.415 -88.3249>
  }
 ]
}
constraints [
 {bone1Name=dyn_back_hair_01, bone2Name=head, swing1LimitDegrees=15, swing2LimitDegrees=10, twistLimitDegrees=1}
 {bone1Name=dyn_back_hair_02, bone2Name=dyn_back_hair_01, swing1LimitDegrees=15, swing2LimitDegrees=10, twistLimitDegrees=1}
 {bone1Name=dyn_front_hair_01, bone2Name=head, swing1LimitDegrees=15, swing2LimitDegrees=20, twistLimitDegrees=1}
 {bone1Name=dyn_front_hair_02, bone2Name=dyn_front_hair_01, swing1LimitDegrees=10, swing2LimitDegrees=20, twistLimitDegrees=5}
 {bone1Name=dyn_mid_hair_01, bone2Name=head, swing1LimitDegrees=10, swing2LimitDegrees=20, twistLimitDegrees=1}
 {bone1Name=dyn_mid_hair_02, bone2Name=dyn_mid_hair_01, swing1LimitDegrees=10, swing2LimitDegrees=20, twistLimitDegrees=1}
 {
  bone1Name=extra_head
  bone2Name=head
  swing1LimitDegrees=45
  swing1Motion=Free
  swing2LimitDegrees=45
  swing2Motion=Free
  twistLimitDegrees=45
  twistMotion=Free
 }
]
